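# Physics asset for the medium male RogueNinja player body that simulates the swinging jacket on its skeleton.
# Seven three-link capsule chains (left and right A–C plus center D) hang from base_jacket anchor bodies, alongside spine and thigh capsules.
# Constraints allow mainly swing2 motion with near-zero swing1 and twist, both locked on the center chain, so each chain bends almost entirely in one plane.
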PhysicsAsset /Game/Characters/Player/Male/Medium/Bodies/M_MED_RogueNinja/Meshes/M_MED_RogueNinja_PhysicsMoving2d
bodies {
 base_jacket_c_D [{shapeType=Capsule, length=0, radius=0, shapeName=base_jacket_c_D_capsule, rotation=<rot 0 -90 0>}]
 base_jacket_l_A [{shapeType=Capsule, length=0.505, radius=0.505, shapeName=base_jacket_l_A_capsule, rotation=<rot 0 -90 0>}]
 base_jacket_l_B [{shapeType=Capsule, length=0.505, radius=0.505, shapeName=base_jacket_l_B_capsule, rotation=<rot 0 -90 0>}]
 base_jacket_l_C [{shapeType=Capsule, length=0.505, radius=0.505, shapeName=base_jacket_l_C_capsule, rotation=<rot 0 -90 0>}]
 base_jacket_r_A [{shapeType=Capsule, length=0.505, radius=0.505, shapeName=base_jacket_r_A_capsule, rotation=<rot 0 -90 0>}]
 base_jacket_r_B [{shapeType=Capsule, length=0.505, radius=0.505, shapeName=base_jacket_r_B_capsule, rotation=<rot 0 -90 0>}]
 base_jacket_r_C [{shapeType=Capsule, length=0.505, radius=0.505, shapeName=base_jacket_r_C_capsule, rotation=<rot 0 -90 0>}]
 dyn_jacket_c_D1 [{shapeType=Capsule, length=1.5, radius=3.5, shapeName=dyn_jacket_r_C1_capsule, center=<-1 0 0>, rotation=<rot 90 0 0>}]
 dyn_jacket_c_D2 [{shapeType=Capsule, length=2, radius=4, shapeName=dyn_jacket_r_C2_capsule, center=<-1 0 0>, rotation=<rot 90 0 0>}]
 dyn_jacket_c_D3 [{shapeType=Capsule, length=5.5, radius=4, shapeName=dyn_jacket_r_C3_capsule, center=<-3 0 0>, rotation=<rot 90 0 0>}]
 dyn_jacket_l_A1 [{shapeType=Capsule, length=3, radius=3, shapeName=dyn_jacket_l_A1_capsule, center=<1 0 0>, rotation=<rot 90 0 0>}]
 dyn_jacket_l_A2 [{shapeType=Capsule, length=3, radius=3, shapeName=dyn_jacket_l_A2_capsule, center=<1 0 0>, rotation=<rot 90 0 0>}]
 dyn_jacket_l_A3 [{shapeType=Capsule, length=7, radius=3, shapeName=dyn_jacket_l_A3_capsule, center=<3 0 0>, rotation=<rot 90 0 0>}]
 dyn_jacket_l_B1 [{shapeType=Capsule, length=3, radius=2, shapeName=dyn_jacket_l_B1_capsule, center=<1 0 0>, rotation=<rot 90 0 0>}]
 dyn_jacket_l_B2 [{shapeType=Capsule, length=3, radius=2, shapeName=dyn_jacket_l_B2_capsule, center=<1 0 0>, rotation=<rot 90 0 0>}]
 dyn_jacket_l_B3 [{shapeType=Capsule, length=7, radius=2, shapeName=dyn_jacket_l_B3_capsule, center=<3 0 0>, rotation=<rot 90 0 0>}]
 dyn_jacket_l_C1 [{shapeType=Capsule, length=3, radius=2, shapeName=dyn_jacket_l_C1_capsule, center=<1 0 0>, rotation=<rot 90 -6.5498 -4.4033>}]
 dyn_jacket_l_C2 [{shapeType=Capsule, length=3, radius=2, shapeName=dyn_jacket_l_C2_capsule, center=<1 0 0>, rotation=<rot 90 -5.4056 -3.2591>}]
 dyn_jacket_l_C3 [{shapeType=Capsule, length=7, radius=2, shapeName=dyn_jacket_l_C3_capsule, center=<3 0 0>, rotation=<rot 90 -8.2236 -6.077>}]
 dyn_jacket_r_A1 [{shapeType=Capsule, length=3, radius=2, shapeName=dyn_jacket_r_A1_capsule, center=<-1 0 0>, rotation=<rot 90 0 0>}]
 dyn_jacket_r_A2 [{shapeType=Capsule, length=3, radius=2, shapeName=dyn_jacket_r_A2_capsule, center=<-1 0 -1.5>, rotation=<rot 90 0 0>}]
 dyn_jacket_r_A3 [{shapeType=Capsule, length=7, radius=3, shapeName=dyn_jacket_r_A3_capsule, center=<-3 0 -1.5>, rotation=<rot 90 0 0>}]
 dyn_jacket_r_B1 [{shapeType=Capsule, length=3, radius=2, shapeName=dyn_jacket_r_B1_capsule, center=<-1 0 0>, rotation=<rot 90 0 0>}]
 dyn_jacket_r_B2 [{shapeType=Capsule, length=3, radius=2, shapeName=dyn_jacket_r_B2_capsule, center=<-1 0 0>, rotation=<rot 90 0 0>}]
 dyn_jacket_r_B3 [{shapeType=Capsule, length=7, radius=3, shapeName=dyn_jacket_r_B3_capsule, center=<-3 0 0>, rotation=<rot 90 0 0>}]
 dyn_jacket_r_C1 [{shapeType=Capsule, length=3, radius=2, shapeName=dyn_jacket_r_C1_capsule, center=<-1 0 0>, rotation=<rot 90 0 0>}]
 dyn_jacket_r_C2 [{shapeType=Capsule, length=3, radius=2, shapeName=dyn_jacket_r_C2_capsule, center=<-1 0 0>, rotation=<rot 90 0 0>}]
 dyn_jacket_r_C3 [{shapeType=Capsule, length=7, radius=2, shapeName=dyn_jacket_r_C3_capsule, center=<-3 0 0>, rotation=<rot 90 0 0>}]
 spine_01 [{shapeType=Capsule, length=8, radius=12, shapeName=spine_01_capsule, center=<5.2463 3.7137 0>, rotation=<rot 90 0 0>}]
 thigh_l [{shapeType=Capsule, length=30, radius=11, shapeName=thigh_l_capsule, center=<-15.7385 1.3 0.3>, rotation=<rot 87.4521 179.9993 179.9995>}]
 thigh_r [{shapeType=Capsule, length=30, radius=11, shapeName=thigh_l_capsule, center=<15.7537 -1.3 -0.3>, rotation=<rot 87.4521 179.9993 -0.0005>}]
}
constraints [
 {bone1Name=dyn_jacket_l_A2, bone2Name=dyn_jacket_l_A1, swing1LimitDegrees=1, swing2LimitDegrees=15, twistLimitDegrees=1}
 {bone1Name=dyn_jacket_l_A3, bone2Name=dyn_jacket_l_A2, swing1LimitDegrees=1, swing2LimitDegrees=15, twistLimitDegrees=1}
 {bone1Name=dyn_jacket_l_B2, bone2Name=dyn_jacket_l_B1, swing1LimitDegrees=1, swing2LimitDegrees=15, twistLimitDegrees=1}
 {bone1Name=dyn_jacket_l_B3, bone2Name=dyn_jacket_l_B2, swing1LimitDegrees=1, swing2LimitDegrees=15, twistLimitDegrees=1}
 {bone1Name=dyn_jacket_l_C2, bone2Name=dyn_jacket_l_C1, swing1LimitDegrees=1, swing2LimitDegrees=15, twistLimitDegrees=1}
 {bone1Name=dyn_jacket_l_C3, bone2Name=dyn_jacket_l_C2, swing1LimitDegrees=1, swing2LimitDegrees=15, twistLimitDegrees=1}
 {bone1Name=dyn_jacket_r_A2, bone2Name=dyn_jacket_r_A1, swing1LimitDegrees=1, swing2LimitDegrees=10, twistLimitDegrees=1}
 {bone1Name=dyn_jacket_r_A3, bone2Name=dyn_jacket_r_A2, swing1LimitDegrees=1, swing2LimitDegrees=10, twistLimitDegrees=1}
 {bone1Name=dyn_jacket_r_B2, bone2Name=dyn_jacket_r_B1, swing1LimitDegrees=1, swing2LimitDegrees=15, twistLimitDegrees=1}
 {bone1Name=dyn_jacket_r_B3, bone2Name=dyn_jacket_r_B2, swing1LimitDegrees=1, swing2LimitDegrees=15, twistLimitDegrees=1}
 {bone1Name=dyn_jacket_r_C2, bone2Name=dyn_jacket_r_C1, swing1LimitDegrees=1, swing2LimitDegrees=15, twistLimitDegrees=1}
 {bone1Name=dyn_jacket_r_C3, bone2Name=dyn_jacket_r_C2, swing1LimitDegrees=1, swing2LimitDegrees=15, twistLimitDegrees=1}
 {bone1Name=dyn_jacket_l_A1, bone2Name=base_jacket_l_A, swing1LimitDegrees=1, swing2LimitDegrees=15, twistLimitDegrees=1}
 {bone1Name=dyn_jacket_l_B1, bone2Name=base_jacket_l_B, swing1LimitDegrees=1, swing2LimitDegrees=15, twistLimitDegrees=1}
 {bone1Name=dyn_jacket_l_C1, bone2Name=base_jacket_l_C, swing1LimitDegrees=1, swing2LimitDegrees=15, twistLimitDegrees=1}
 {bone1Name=dyn_jacket_r_A1, bone2Name=base_jacket_r_A, swing1LimitDegrees=1, swing2LimitDegrees=15, twistLimitDegrees=1}
 {bone1Name=dyn_jacket_r_B1, bone2Name=base_jacket_r_B, swing1LimitDegrees=1, swing2LimitDegrees=15, twistLimitDegrees=1}
 {bone1Name=dyn_jacket_r_C1, bone2Name=base_jacket_r_C, swing1LimitDegrees=1, swing2LimitDegrees=15, twistLimitDegrees=1}
 {
  bone1Name=dyn_jacket_c_D1
  bone2Name=base_jacket_c_D
  swing1LimitDegrees=45
  swing1Motion=Locked
  swing2LimitDegrees=12
  twistLimitDegrees=45
  twistMotion=Locked
 }
 {
  bone1Name=dyn_jacket_c_D2
  bone2Name=dyn_jacket_c_D1
  swing1LimitDegrees=45
  swing1Motion=Locked
  swing2LimitDegrees=8
  twistLimitDegrees=45
  twistMotion=Locked
 }
 {
  bone1Name=dyn_jacket_c_D3
  bone2Name=dyn_jacket_c_D2
  swing1LimitDegrees=45
  swing1Motion=Locked
  swing2LimitDegrees=10
  twistLimitDegrees=45
  twistMotion=Locked
 }
]
