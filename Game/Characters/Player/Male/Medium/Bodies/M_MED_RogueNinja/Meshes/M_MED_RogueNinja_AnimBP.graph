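# Post-copy AnimBlueprint for the Rogue Ninja player body, animating jacket, hood, ribbons and bags over the attached parent mesh's pose.
# A Control Rig adjusts the copied pose, feeding cached state branches that combine Look At nodes, constraints, additive bone tweaks and RigidBody simulation with per-state physics assets.
# Bool blends select states like DBNO, skydiving or emotes, then Anim Dynamics and variable-driven hood and corrective bone fixes finish the pose.
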AnimBlueprint /Game/Characters/Player/Male/Medium/Bodies/M_MED_RogueNinja/Meshes/M_MED_RogueNinja_AnimBP
target_skeleton=&/M_MED_RogueNinja_Skeleton
graphs [
 AnimGraph {nodes [
   Look At - Bone: dyn_jacket_l_A1(Alpha=1) {
    alphaBoolBlend {blendOption=Linear}
    boneToModify {boneName=dyn_jacket_l_A1}
    lookAtLocation=<-15.1072 2.7105 -24.2247>
    lookAtTarget {boneReference {boneName=deform_knee_out_l}}
    lookAt_Axis {axis=<1 0>}
   }
   n1 = Get Alpha_isSkydiveFloating_RBDisable() -> (Alpha_isSkydiveFloating_RBDisable:float)
   "Transform (Modify) Bone - Bone: base_jacket_l_C"(Alpha=@n1, Translation="0,0,0", Rotation="5,0,0", Scale="1,1,1") {
    rotationMode=BMM_Additive
    rotationSpace=BCS_BoneSpace
    translationMode=BMM_Additive
    alphaBoolBlend {blendOption=Linear}
    alphaScaleBiasClamp {bInterpResult=true, interpSpeedIncreasing=5}
    boneToModify {boneName=base_jacket_l_C}
    rotation {pitch=5}
   }
   n2 = Blend Poses by bool(
    BlendPose_0=Use cached pose 'SlidingCache'()
    BlendPose_1=Blend Poses by bool(BlendPose_0=Use cached pose 'CrouchCache'(), BlendPose_1=Use cached pose 'Moving2DCache'(), BlendTime_0=0.2, bActiveValue=Get bIsCrouching()) {blendType=Linear}
    BlendTime_0=0.4
    bActiveValue=Get bIsKneeSliding()
   ) {blendType=Linear}
   n3 = Use cached pose 'SlidingWeapons'()
   n4 = Get Alpha_SlidingWeapon() -> (Alpha_SlidingWeapon:bool)
   n5 = Blend Poses by bool(
    BlendPose_0=Use cached pose 'ZiplineCache'()
    BlendPose_1=Blend Poses by bool(
     BlendPose_0=Use cached pose 'SkyDivingCache'()
     BlendPose_1=Blend Poses by bool(BlendPose_0=@n3, BlendPose_1=@n2, BlendTime_0=0.5, bActiveValue=@n4) {blendType=Linear}
     BlendTime_0=0.5
     bActiveValue=Get bIsSkyDiving()
    ) {blendType=Linear}
    BlendTime_0=0.5
    bActiveValue=Get OnZipline() -> (OnZipline:bool)
   ) {blendType=Linear}
   n6 = Use cached pose 'OnRailCache'()
   n7 = Get OnRailMode() -> (OnRailMode:bool)
   n8 = Blend Poses by bool(
    BlendPose_0=Use cached pose 'Moving2DCache'()
    BlendPose_1=Blend Poses by bool(
     BlendPose_0=Use cached pose 'DBNOCache'()
     BlendPose_1=Blend Poses by bool(BlendPose_0=@n6, BlendPose_1=@n5, BlendTime_0=0.5, bActiveValue=@n7) {blendType=Linear}
     BlendTime_0=0.5
     bActiveValue=Get bIsDBNO()
    ) {blendType=Linear}
    BlendTime_0=0.5
    bActiveValue=Get bIsInFrontEnd()
   ) {blendType=Linear}
   n9 = Use cached pose 'EmoteCache'()
   n10 = Get bIsPlayingEmote()
   n11 = Anim Dynamics - Bone: dyn_l_ribbonB(
    ComponentPose=Local To Component(LocalPose=Blend Poses by bool(BlendPose_0=@n9, BlendPose_1=@n8, BlendTime_0=0.8, BlendTime_1=0.8, bActiveValue=@n10) {blendType=Linear})
    Alpha=1
    ExternalForce="0,0,0"
   ) {
    angularDampingOverride=1
    angularSpringConstant=20000
    bAngularSpring=true
    bOverrideAngularDamping=true
    bUseGravityOverride=true
    gravityScale=0.35
    simulationSpace=RootRelative
    alphaBoolBlend {blendOption=Linear}
    boundBone {boneName=dyn_l_ribbonB}
    componentLinearAccScale=<1 1 1>
    componentLinearVelScale=<1 1 1>
    gravityOverride {z=-980}
    physicsBodyDefinitions [{
      boundBone {boneName=dyn_l_ribbonB}
      boxExtents=<30 30 30>
      constraintSetup {angularLimitsMax=<0 0 60>, angularLimitsMin=<0 0 -60>, angularTarget=<1 0 0>}
      localJointOffset=<2 0 0>
     }]
   }
   n12 = "Transform (Modify) Bone - Bone: dyn_l_ribbonB"(ComponentPose=@n11, Alpha=1, Translation="0,50,0", Rotation="-18.370,0.000,-0.000", Scale="1,1,1") {
    rotationMode=BMM_Additive
    rotationSpace=BCS_BoneSpace
    alphaBoolBlend {blendOption=Linear}
    alphaScaleBiasClamp {bInterpResult=true}
    boneToModify {boneName=dyn_l_ribbonB}
    rotation=<rot -18.3699 0 0>
    translation {y=50}
   }
   n13 = Anim Dynamics - Bone: dyn_l_ribbonA(ComponentPose=@n12, Alpha=1, ExternalForce="0,0,0") {
    angularDampingOverride=1
    angularSpringConstant=20000
    bAngularSpring=true
    bOverrideAngularDamping=true
    bUseGravityOverride=true
    gravityScale=0.35
    simulationSpace=RootRelative
    alphaBoolBlend {blendOption=Linear}
    boundBone {boneName=dyn_l_ribbonA}
    componentLinearAccScale=<1 1 1>
    componentLinearVelScale=<1 1 1>
    gravityOverride {z=-980}
    physicsBodyDefinitions [{
      boundBone {boneName=dyn_l_ribbonA}
      boxExtents=<30 30 30>
      constraintSetup {angularLimitsMax=<0 0 75>, angularLimitsMin=<0 0 -75>, angularTarget=<1 0 0>}
      localJointOffset=<2 0 0>
     }]
    relativeSpaceBone {boneName=upperarm_l}
   }
   n14 = "Transform (Modify) Bone - Bone: dyn_l_ribbonA"(ComponentPose=@n13, Alpha=1, Translation="0,50,0", Rotation="-18.268,6.265,-1.972", Scale="1,1,1") {
    rotationMode=BMM_Additive
    rotationSpace=BCS_BoneSpace
    alphaBoolBlend {blendOption=Linear}
    alphaScaleBiasClamp {bInterpResult=true}
    boneToModify {boneName=dyn_l_ribbonA}
    rotation=<rot -18.268 6.265 -1.972>
    translation {y=50}
   }
   n15 = "Transform (Modify) Bone - Bone: dyn_l_ribbonA"(ComponentPose=@n14, Alpha=Get TargetingUp() -> (TargetingUp:double), Translation="0,50,0", Rotation="21.026,-0.000,0.002", Scale="1,1,1") {
    rotationMode=BMM_Additive
    rotationSpace=BCS_BoneSpace
    alphaBoolBlend {blendOption=Linear}
    alphaScaleBiasClamp {bInterpResult=true}
    boneToModify {boneName=dyn_l_ribbonA}
    rotation {pitch=21.026, roll=0.002}
    translation {y=50}
   }
   n16 = Anim Dynamics - Bone: dyn_l_bagA(ComponentPose=@n15, Alpha=1, ExternalForce="0,0,0") {
    angularDampingOverride=1
    angularSpringConstant=15000
    bAngularSpring=true
    bOverrideAngularDamping=true
    bUseGravityOverride=true
    gravityScale=0.35
    simulationSpace=RootRelative
    alphaBoolBlend {blendOption=Linear}
    boundBone {boneName=dyn_l_bagA}
    componentLinearAccScale=<1 1 1>
    componentLinearVelScale=<1 1 1>
    gravityOverride {z=-980}
    physicsBodyDefinitions [{
      boundBone {boneName=dyn_l_bagA}
      boxExtents=<10 20 20>
      constraintSetup {angularLimitsMax=<0 0 45>, angularLimitsMin=<0 0 -10>, angularTarget=<1 0 0>}
      localJointOffset=<0.8 0 0>
     }]
    relativeSpaceBone {boneName=thigh_l}
   }
   n17 = Anim Dynamics - Bone: dyn_l_bagB(ComponentPose=@n16, Alpha=1, ExternalForce="0,0,0") {
    angularDampingOverride=1
    angularSpringConstant=15000
    bAngularSpring=true
    bOverrideAngularDamping=true
    bUseGravityOverride=true
    gravityScale=0.35
    simulationSpace=RootRelative
    alphaBoolBlend {blendOption=Linear}
    boundBone {boneName=dyn_l_bagB}
    componentLinearAccScale=<1 1 1>
    componentLinearVelScale=<1 1 1>
    gravityOverride {z=-980}
    physicsBodyDefinitions [{
      boundBone {boneName=dyn_l_bagB}
      boxExtents=<10 20 20>
      constraintSetup {angularLimitsMax=<0 0 45>, angularLimitsMin=<0 0 -20>, angularTarget=<1 0 0>}
      localJointOffset=<0.8 0 0>
     }]
    relativeSpaceBone {boneName=thigh_l}
   }
   n18 = Anim Dynamics - Bone: dyn_r_bagB(ComponentPose=@n17, Alpha=1, ExternalForce="0,0,0") {
    angularDampingOverride=1
    angularSpringConstant=15000
    bAngularSpring=true
    bOverrideAngularDamping=true
    bUseGravityOverride=true
    gravityScale=0.35
    simulationSpace=RootRelative
    alphaBoolBlend {blendOption=Linear}
    boundBone {boneName=dyn_r_bagB}
    componentLinearAccScale=<1 1 1>
    componentLinearVelScale=<1 1 1>
    gravityOverride {z=-980}
    physicsBodyDefinitions [{
      boundBone {boneName=dyn_r_bagB}
      boxExtents=<10 20 20>
      constraintSetup {angularLimitsMax=<0 0 45>, angularLimitsMin=<0 0 -20>, angularTarget=<1 0 0>}
      localJointOffset=<0.8 0 0>
     }]
    relativeSpaceBone {boneName=thigh_r}
   }
   n19 = Anim Dynamics - Bone: dyn_r_bagA(ComponentPose=@n18, Alpha=1, ExternalForce="0,0,0") {
    angularDampingOverride=1
    angularSpringConstant=15000
    bAngularSpring=true
    bOverrideAngularDamping=true
    bUseGravityOverride=true
    gravityScale=0.35
    simulationSpace=RootRelative
    alphaBoolBlend {blendOption=Linear}
    boundBone {boneName=dyn_r_bagA}
    componentLinearAccScale=<1 1 1>
    componentLinearVelScale=<1 1 1>
    gravityOverride {z=-980}
    physicsBodyDefinitions [{
      boundBone {boneName=dyn_r_bagA}
      boxExtents=<10 20 20>
      constraintSetup {angularLimitsMax=<0 0 45>, angularLimitsMin=<0 0 -10>, angularTarget=<1 0 0>}
      localJointOffset=<0.8 0 0>
     }]
    relativeSpaceBone {boneName=thigh_r}
   }
   n20 = "Constraint - hood_r (P)"(ComponentPose=@n19, Alpha=Get Alpha_isHeadRotateR() -> (Alpha_isHeadRotateR:float), ConstraintWeights_0=1) {
    alphaBoolBlend {blendOption=Linear}
    alphaScaleBiasClamp {bInterpResult=true, interpSpeedDecreasing=3, interpSpeedIncreasing=3}
    boneToModify {boneName=hood_r}
    constraintSetup [{targetBone {boneName=head}}]
    constraintWeights [1]
   }
   n21 = "Transform (Modify) Bone - Bone: hood_low_r"(
    ComponentPose=@n20
    Alpha=Get Alpha_IsHeadRotateHoodCorrect_L() -> (Alpha_IsHeadRotateHoodCorrect_L:double)
    Translation="-4.758,5.996,0.018"
    Rotation="0,0,0"
    Scale="1,1,1"
   ) {
    rotationMode=BMM_Additive
    rotationSpace=BCS_BoneSpace
    translationMode=BMM_Additive
    translationSpace=BCS_BoneSpace
    alphaBoolBlend {blendOption=Linear}
    alphaScaleBiasClamp {bInterpResult=true}
    boneToModify {boneName=hood_low_r}
    translation=<-4.758 5.996 0.018>
   }
   n22 = "Transform (Modify) Bone - Bone: hood_low_l"(
    ComponentPose=@n21
    Alpha=Get Alpha_IsHeadRotateHoodCorrect_R() -> (Alpha_IsHeadRotateHoodCorrect_R:double)
    Translation="5,0,0"
    Rotation="0,0,0"
    Scale="1,1,1"
   ) {
    rotationMode=BMM_Additive
    rotationSpace=BCS_BoneSpace
    translationMode=BMM_Additive
    translationSpace=BCS_BoneSpace
    alphaBoolBlend {blendOption=Linear}
    alphaScaleBiasClamp {bInterpResult=true}
    boneToModify {boneName=hood_low_l}
    translation {x=5}
   }
   n23 = "Constraint - hood_l (P)"(ComponentPose=@n22, Alpha=Get Alpha_isHeadRotateL() -> (Alpha_isHeadRotateL:float), ConstraintWeights_0=1) {
    alphaBoolBlend {blendOption=Linear}
    alphaScaleBiasClamp {bInterpResult=true, interpSpeedDecreasing=3, interpSpeedIncreasing=3}
    boneToModify {boneName=hood_l}
    constraintSetup [{targetBone {boneName=neck_01}}]
    constraintWeights [1]
   }
   n24 = "Transform (Modify) Bone - Bone: clavicle_correct_r"(
    ComponentPose=@n23
    Alpha=Get Is SkyDivingFrontEnd() -> (Is SkyDivingFrontEnd:double)
    Translation="-1,-1,-4"
    Rotation="0.000,-1.629,0.000"
    Scale="1,1,1"
   ) {
    rotationMode=BMM_Additive
    rotationSpace=BCS_BoneSpace
    translationMode=BMM_Additive
    translationSpace=BCS_BoneSpace
    alphaBoolBlend {blendOption=Linear}
    alphaScaleBiasClamp {bInterpResult=true}
    boneToModify {boneName=clavicle_correct_r}
    rotation {yaw=-1.629}
    translation=<-1 -1 -4>
   }
   n25 = "Transform (Modify) Bone - Bone: dyn_jacket_l_A3"(ComponentPose=@n24, Alpha=Get Is FloatingFrontEnd() -> (Is FloatingFrontEnd:double), Translation="2,0,0", Rotation="0,20,0", Scale="1,1,1") {
    rotationMode=BMM_Additive
    rotationSpace=BCS_BoneSpace
    translationMode=BMM_Additive
    translationSpace=BCS_BoneSpace
    alphaBoolBlend {blendOption=Linear}
    alphaScaleBiasClamp {bInterpResult=true}
    boneToModify {boneName=dyn_jacket_l_A3}
    rotation {yaw=20}
    translation {x=2}
   }
   n26 = "Transform (Modify) Bone - Bone: hood_r"(
    ComponentPose=@n25
    Alpha=Get Is BowDetected() -> (Is BowDetected:double)
    Translation="-0.001,-2.405,-0.078"
    Rotation="-11.520,0.000,-0.000"
    Scale="1,1,1"
   ) {
    rotationMode=BMM_Additive
    rotationSpace=BCS_BoneSpace
    translationMode=BMM_Additive
    translationSpace=BCS_BoneSpace
    alphaBoolBlend {blendOption=Linear}
    alphaScaleBiasClamp {bInterpResult=true, interpSpeedDecreasing=3, interpSpeedIncreasing=3}
    boneToModify {boneName=hood_r}
    rotation=<rot -11.52 0 0>
    translation=<-0.001 -2.405 -0.078>
   }
   n27 = "Transform (Modify) Bone - Bone: hood_r"!(ComponentPose=@n26, bAlphaBoolEnabled:bool=Get bIsSurfaceSwimming(), Translation="-3,0,0", Rotation="0,0,0", Scale="1,1,1") -> (Pose) {
    alphaInputType=Bool
    rotationMode=BMM_Additive
    rotationSpace=BCS_BoneSpace
    translationMode=BMM_Additive
    translationSpace=BCS_BoneSpace
    alphaBoolBlend {blendOption=Linear}
    alphaScaleBiasClamp {bInterpResult=true, interpSpeedDecreasing=3, interpSpeedIncreasing=3}
    boneToModify {boneName=hood_r}
    translation {x=-3}
   }
   n28 = "Transform (Modify) Bone - Bone: back_correct_scale"(ComponentPose=@n27, Alpha=Get Is Melee_moving2d() -> (Is Melee_moving2d:double), Translation="0,10,0", Rotation="0,0,0", Scale="0.3,0.3,0.3") {
    rotationMode=BMM_Additive
    rotationSpace=BCS_BoneSpace
    scaleMode=BMM_Additive
    scaleSpace=BCS_BoneSpace
    translationMode=BMM_Additive
    translationSpace=BCS_BoneSpace
    alphaBoolBlend {blendOption=Linear}
    alphaScaleBiasClamp {bInterpResult=true, interpSpeedDecreasing=3, interpSpeedIncreasing=3}
    boneToModify {boneName=back_correct_scale}
    scale=<0.3 0.3 0.3>
    translation {y=10}
   }
   n29 = "Transform (Modify) Bone - Bone: spine_05_correct_l"(
    ComponentPose=@n28
    Alpha=Get Alpha_IsMeleeDualWeild() -> (Alpha_IsMeleeDualWeild:double)
    Translation="-3,2,1.5"
    Rotation="0,0,0"
    Scale="1,1,1"
   ) {
    rotationMode=BMM_Additive
    rotationSpace=BCS_BoneSpace
    scaleMode=BMM_Additive
    scaleSpace=BCS_BoneSpace
    translationMode=BMM_Additive
    translationSpace=BCS_BoneSpace
    alphaBoolBlend {blendOption=Linear}
    alphaScaleBiasClamp {bInterpResult=true, interpSpeedDecreasing=3, interpSpeedIncreasing=3}
    boneToModify {boneName=spine_05_correct_l}
    translation=<-3 2 1.5>
   }
   n30 = "Transform (Modify) Bone - Bone: spine_05_correct_l"(ComponentPose=@n29, Alpha=Get Alpha_openParachute() -> (Alpha_openParachute:double), Translation="0,0,1.2", Rotation="0,0,0", Scale="1,1,1") {
    rotationMode=BMM_Additive
    rotationSpace=BCS_BoneSpace
    scaleMode=BMM_Additive
    scaleSpace=BCS_BoneSpace
    translationMode=BMM_Additive
    translationSpace=BCS_BoneSpace
    alphaBoolBlend {blendOption=Linear}
    alphaScaleBiasClamp {bInterpResult=true, interpSpeedDecreasing=3, interpSpeedIncreasing=3}
    boneToModify {boneName=spine_05_correct_l}
    translation {z=1.2}
   }
   Output Pose(Result=Component To Local(ComponentPose=@n30)) {layerGroup=None}
   n31 = Look At - Bone: dyn_jacket_l_A1(ComponentPose=Local To Component(LocalPose=Use cached pose 'ControlRigCache'()), Alpha=1) {
    alphaBoolBlend {blendOption=Linear}
    boneToModify {boneName=dyn_jacket_l_A1}
    lookAtLocation=<8.6146 0.1632 -7.6098>
    lookAtTarget {boneReference {boneName=deform_knee_out_l}}
    lookAt_Axis {axis=<1 0>}
   }
   n32 = Look At - Bone: dyn_jacket_l_B1(ComponentPose=@n31, Alpha=1) {
    alphaBoolBlend {blendOption=Linear}
    boneToModify {boneName=dyn_jacket_l_B1}
    lookAtLocation=<-17.5476 -6.3029 4.2338>
    lookAtTarget {boneReference {boneName=dyn_l_bagB}}
    lookAt_Axis {axis=<1 0>}
   }
   n33 = Look At - Bone: dyn_jacket_r_A1(ComponentPose=@n32, Alpha=1) {
    alphaBoolBlend {blendOption=Linear}
    boneToModify {boneName=dyn_jacket_r_A1}
    lookAtLocation=<16.5461 -19.492 6.0732>
    lookAtTarget {boneReference {boneName=deform_knee_out_r}}
    lookAt_Axis {axis=<-1 0>}
   }
   n34 = Look At - Bone: dyn_jacket_r_B1(ComponentPose=@n33, Alpha=1) {
    alphaBoolBlend {blendOption=Linear}
    boneToModify {boneName=dyn_jacket_r_B1}
    lookAtLocation=<11.446 26.4383 26.1>
    lookAtTarget {boneReference {boneName=deform_knee_out_r}}
    lookAt_Axis {axis=<-1 0>}
   }
   n35 = "Constraint - base_jacket_l_C (RR)"!(ComponentPose=@n34, Alpha=1, ConstraintWeights_0=1, ConstraintWeights_1:float=1) -> (Pose) {
    alphaBoolBlend {blendOption=Linear}
    alphaScaleBiasClamp {bInterpResult=true, interpSpeedDecreasing=4, interpSpeedIncreasing=4}
    boneToModify {boneName=base_jacket_l_C}
    constraintSetup [{transformType=Rotation, targetBone {boneName=dyn_jacket_c_D1}}, {transformType=Rotation, targetBone {boneName=dyn_jacket_l_B1}}]
    constraintWeights [1,1]
   }
   n36 = "Constraint - dyn_jacket_r_C1 (RR)"!(ComponentPose=@n35, Alpha=1, ConstraintWeights_0=1, ConstraintWeights_1:float=0.5) -> (Pose) {
    alphaBoolBlend {blendOption=Linear}
    boneToModify {boneName=dyn_jacket_r_C1}
    constraintSetup [{transformType=Rotation, targetBone {boneName=dyn_jacket_c_D1}}, {transformType=Rotation, targetBone {boneName=dyn_jacket_r_B1}}]
    constraintWeights [1,1]
   }
   n37 = "Transform (Modify) Bone - Bone: base_jacket_l_A"(ComponentPose=@n36, Alpha=Get Alpha_IsDBNOCarried() -> (Alpha_IsDBNOCarried:double), Translation="0,0,0", Rotation="15,-25,0", Scale="1,1,1") {
    rotationMode=BMM_Additive
    rotationSpace=BCS_BoneSpace
    scaleMode=BMM_Additive
    scaleSpace=BCS_BoneSpace
    translationMode=BMM_Additive
    translationSpace=BCS_BoneSpace
    alphaBoolBlend {blendOption=Linear}
    alphaScaleBiasClamp {bInterpResult=true, interpSpeedDecreasing=3, interpSpeedIncreasing=4}
    boneToModify {boneName=base_jacket_l_A}
    rotation {pitch=15, yaw=-25}
   }
   n38 = Look At - Bone: dyn_jacket_l_A1(ComponentPose=Local To Component(LocalPose=Use cached pose 'ControlRigCache'()), Alpha=1) {
    alphaBoolBlend {blendOption=Linear}
    boneToModify {boneName=dyn_jacket_l_A1}
    lookAtLocation=<-15.1072 2.7105 -28.4409>
    lookAtTarget {boneReference {boneName=deform_knee_out_l}}
    lookAt_Axis {axis=<1 0>}
   }
   n39 = Look At - Bone: dyn_jacket_l_B1(ComponentPose=@n38, Alpha=1) {
    alphaBoolBlend {blendOption=Linear}
    boneToModify {boneName=dyn_jacket_l_B1}
    lookAtLocation=<-17.5476 -6.3029 4.2338>
    lookAtTarget {boneReference {boneName=dyn_l_bagB}}
    lookAt_Axis {axis=<1 0>}
   }
   n40 = Look At - Bone: dyn_jacket_r_A1(ComponentPose=@n39, Alpha=1) {
    alphaBoolBlend {blendOption=Linear}
    boneToModify {boneName=dyn_jacket_r_A1}
    lookAtLocation=<16.5461 -19.492 6.0732>
    lookAtTarget {boneReference {boneName=deform_knee_out_r}}
    lookAt_Axis {axis=<-1 0>}
   }
   n41 = Look At - Bone: dyn_jacket_r_B1(ComponentPose=@n40, Alpha=1) {
    alphaBoolBlend {blendOption=Linear}
    boneToModify {boneName=dyn_jacket_r_B1}
    lookAtLocation=<11.446 26.4383 26.1>
    lookAtTarget {boneReference {boneName=deform_knee_out_r}}
    lookAt_Axis {axis=<-1 0>}
   }
   n42 = "Constraint - base_jacket_l_C (RR)"(ComponentPose=@n41, Alpha=1, ConstraintWeights_0=1, ConstraintWeights_1=1) {
    alphaBoolBlend {blendOption=Linear}
    alphaScaleBiasClamp {bInterpResult=true, interpSpeedDecreasing=4, interpSpeedIncreasing=4}
    boneToModify {boneName=base_jacket_l_C}
    constraintSetup [{transformType=Rotation, targetBone {boneName=dyn_jacket_c_D1}}, {transformType=Rotation, targetBone {boneName=dyn_jacket_l_B1}}]
    constraintWeights [1,1]
   }
   n43 = "Constraint - dyn_jacket_r_C1 (RR)"(ComponentPose=@n42, Alpha=1, ConstraintWeights_0=1, ConstraintWeights_1=0.5) {
    alphaBoolBlend {blendOption=Linear}
    boneToModify {boneName=dyn_jacket_r_C1}
    constraintSetup [{transformType=Rotation, targetBone {boneName=dyn_jacket_c_D1}}, {transformType=Rotation, targetBone {boneName=dyn_jacket_r_B1}}]
    constraintWeights [1,1]
   }
   n44 = Get bIsDBNO()
   n45 = "Transform (Modify) Bone - Bone: base_jacket_l_B"!(ComponentPose=@n43, bAlphaBoolEnabled:bool=@n44, Translation="-1.5,0,0", Rotation="0,12.5,0", Scale="1,1,1") -> (Pose) {
    alphaInputType=Bool
    rotationMode=BMM_Additive
    rotationSpace=BCS_BoneSpace
    scaleMode=BMM_Additive
    scaleSpace=BCS_BoneSpace
    translationMode=BMM_Additive
    translationSpace=BCS_BoneSpace
    alphaBoolBlend {blendOption=Linear}
    alphaScaleBiasClamp {bInterpResult=true, interpSpeedDecreasing=3, interpSpeedIncreasing=4}
    boneToModify {boneName=base_jacket_l_B}
    rotation {yaw=12.5}
    translation {x=-1.5}
   }
   n46 = "Transform (Modify) Bone - Bone: base_jacket_r_B"!(ComponentPose=@n45, bAlphaBoolEnabled:bool=@n44, Translation="0,0,0", Rotation="0,15,0", Scale="1,1,1") -> (Pose) {
    alphaInputType=Bool
    rotationMode=BMM_Additive
    rotationSpace=BCS_BoneSpace
    scaleMode=BMM_Additive
    scaleSpace=BCS_BoneSpace
    translationMode=BMM_Additive
    translationSpace=BCS_BoneSpace
    alphaBoolBlend {blendOption=Linear}
    alphaScaleBiasClamp {bInterpResult=true, interpSpeedDecreasing=3, interpSpeedIncreasing=4}
    boneToModify {boneName=base_jacket_r_B}
    rotation {yaw=15}
   }
   n47 = Get Alpha_IsDBNOMoving() -> (Alpha_IsDBNOMoving:double)
   n48 = "Transform (Modify) Bone - Bone: base_jacket_l_A"(ComponentPose=@n46, Alpha=@n47, Translation="-6,0,0", Rotation="40,10,0", Scale="1,1,1") {
    rotationMode=BMM_Additive
    rotationSpace=BCS_BoneSpace
    scaleMode=BMM_Additive
    scaleSpace=BCS_BoneSpace
    translationMode=BMM_Additive
    translationSpace=BCS_BoneSpace
    alphaBoolBlend {blendOption=Linear}
    alphaScaleBiasClamp {bInterpResult=true, interpSpeedDecreasing=3, interpSpeedIncreasing=4}
    boneToModify {boneName=base_jacket_l_A}
    rotation {pitch=40, yaw=10}
    translation {x=-6}
   }
   n49 = "Transform (Modify) Bone - Bone: base_jacket_l_B"!(ComponentPose=@n48, Alpha:float=@n47, Translation="0,0,0", Rotation="0,20,0", Scale="1,1,1") -> (Pose) {
    rotationMode=BMM_Additive
    rotationSpace=BCS_BoneSpace
    scaleMode=BMM_Additive
    scaleSpace=BCS_BoneSpace
    translationMode=BMM_Additive
    translationSpace=BCS_BoneSpace
    alphaBoolBlend {blendOption=Linear}
    alphaScaleBiasClamp {bInterpResult=true, interpSpeedDecreasing=3, interpSpeedIncreasing=4}
    boneToModify {boneName=base_jacket_l_B}
    rotation {yaw=20}
   }
   n50 = "Transform (Modify) Bone - Bone: base_jacket_l_C"(
    ComponentPose=@n49
    Alpha=Get Alpha_IsDBNOMovingRB Enabled() -> (Alpha_IsDBNOMovingRB Enabled:double)
    Translation="0,0,0"
    Rotation="30,25,0"
    Scale="1,1,1"
   ) {
    rotationMode=BMM_Additive
    rotationSpace=BCS_BoneSpace
    scaleMode=BMM_Additive
    scaleSpace=BCS_BoneSpace
    translationMode=BMM_Additive
    translationSpace=BCS_BoneSpace
    alphaBoolBlend {blendOption=Linear}
    alphaScaleBiasClamp {bInterpResult=true, interpSpeedDecreasing=3}
    boneToModify {boneName=base_jacket_l_C}
    rotation {pitch=30, yaw=25}
   }
   n51 = RigidBody!(ComponentPose=@n50, Alpha, OverrideWorldGravity:Vector=Get Gravity Override(), ExternalForce) -> (Pose) {
    bOverrideWorldGravity=true
    overridePhysicsAsset=&/M_MED_RogueNinja_PhysicsDBNO
    simulationSpace=BaseBoneSpace
    alphaBoolBlend {blendOption=Linear}
    alphaScaleBiasClamp {bInterpResult=true}
    baseBoneRef {boneName=pelvis}
    componentAppliedLinearAccClamp=<3000 3000 3000>
    componentLinearAccScale=<0.4 0.4 0.4>
    componentLinearVelScale=<0.4 0.4 0.4>
    simSpaceSettings {maxAngularAcceleration=1000, maxAngularVelocity=3, maxLinearAcceleration=700, maxLinearVelocity=500, velocityScaleZ=0, worldAlpha=0.5}
   }
   n52 = "Transform (Modify) Bone - Bone: dyn_jacket_l_B3"(ComponentPose=@n51, Alpha=@n47, Translation="-4,0,0", Rotation="8.603,-1.508,-10.124", Scale="1,1,1") {
    rotationMode=BMM_Additive
    rotationSpace=BCS_BoneSpace
    scaleMode=BMM_Additive
    scaleSpace=BCS_BoneSpace
    translationMode=BMM_Additive
    translationSpace=BCS_BoneSpace
    alphaBoolBlend {blendOption=Linear}
    alphaScaleBiasClamp {bInterpResult=true, interpSpeedDecreasing=3}
    boneToModify {boneName=dyn_jacket_l_B3}
    rotation=<rot 8.603 -1.508 -10.124>
    translation {x=-4}
   }
   n53 = "Transform (Modify) Bone - Bone: dyn_jacket_l_A2"(ComponentPose=@n52, Alpha=1, Translation="-1.376,-0.626,0.011", Rotation="4.121,-11.752,-29.513", Scale="1,1,1") {
    rotationMode=BMM_Additive
    rotationSpace=BCS_BoneSpace
    scaleMode=BMM_Additive
    scaleSpace=BCS_BoneSpace
    translationMode=BMM_Additive
    translationSpace=BCS_BoneSpace
    alphaBoolBlend {blendOption=Linear}
    alphaScaleBiasClamp {bInterpResult=true, interpSpeedDecreasing=3}
    boneToModify {boneName=dyn_jacket_l_A2}
    rotation=<rot 4.121 -11.752 -29.513>
    translation=<-1.376 -0.626 0.011>
   }
   n54 = "Transform (Modify) Bone - Bone: dyn_jacket_r_A2"(ComponentPose=@n53, Alpha=1, Translation="0.429,0.471,-1.316", Rotation="19.621,14.953,-15.217", Scale="1,1,1") {
    rotationMode=BMM_Additive
    rotationSpace=BCS_BoneSpace
    scaleMode=BMM_Additive
    scaleSpace=BCS_BoneSpace
    translationMode=BMM_Additive
    translationSpace=BCS_BoneSpace
    alphaBoolBlend {blendOption=Linear}
    alphaScaleBiasClamp {bInterpResult=true, interpSpeedDecreasing=3}
    boneToModify {boneName=dyn_jacket_r_A2}
    rotation=<rot 19.621 14.953 -15.217>
    translation=<0.429 0.471 -1.316>
   }
   n55 = Get Is DBNOIdle() -> (Is DBNOIdle:double)
   n56 = "Transform (Modify) Bone - Bone: dyn_jacket_l_A2"(ComponentPose=@n54, Alpha=@n55, Translation="0,0,0", Rotation="-15,10,10", Scale="1,1,1") {
    rotationMode=BMM_Additive
    rotationSpace=BCS_BoneSpace
    scaleMode=BMM_Additive
    scaleSpace=BCS_BoneSpace
    translationMode=BMM_Additive
    translationSpace=BCS_BoneSpace
    alphaBoolBlend {blendOption=Linear}
    alphaScaleBiasClamp {bInterpResult=true, interpSpeedDecreasing=3}
    boneToModify {boneName=dyn_jacket_l_A2}
    rotation=<rot -15 10 10>
   }
   n57 = "Transform (Modify) Bone - Bone: dyn_jacket_l_B2"(ComponentPose=@n56, Alpha=@n55, Translation="0,0,0", Rotation="-15,0,-12", Scale="1,1,1") {
    rotationMode=BMM_Additive
    rotationSpace=BCS_BoneSpace
    scaleMode=BMM_Additive
    scaleSpace=BCS_BoneSpace
    translationMode=BMM_Additive
    translationSpace=BCS_BoneSpace
    alphaBoolBlend {blendOption=Linear}
    alphaScaleBiasClamp {bInterpResult=true, interpSpeedDecreasing=3}
    boneToModify {boneName=dyn_jacket_l_B2}
    rotation {pitch=-15, roll=-12}
   }
   Save cached pose 'DBNOCache'(Pose=Blend Poses by bool(
     BlendPose_0=Component To Local(ComponentPose=@n57)
     BlendPose_1=Component To Local(ComponentPose=@n37)
     BlendTime_0=0.5
     bActiveValue=Get bRigidBodyEnabled()
    ) {blendType=Linear}) {cacheName=DBNOCache}
   n58 = Control Rig!(
    Source=Copy Pose From Mesh() {bUseAttachedParent=true}
    Alpha
    AssaultRifle:double=Get AlphaIsAssaultRifle_Float() -> (AlphaIsAssaultRifle_Float:double)
    IsFloating:double=Get Alpha_isSkydiveFloating() -> (Alpha_isSkydiveFloating:float)
    IsOpenParachute:double=Get Alpha_openParachute()
    IsRocketLauncher:double=Get is RocketLauncher_Float() -> (is RocketLauncher_Float:double)
    IsRocketLauncherCoat_Correct:double=Get Is RocketLauncherCorrect() -> (Is RocketLauncherCorrect:double)
    IsRocketLauncherStand:double=Get Is RocketLauncherStand() -> (Is RocketLauncherStand:double)
    Melee:double=Get Alpha_IsMeleeDualWeild()
    Emote:double=Get Is Emote() -> (Is Emote:double)
    MinigunSelected:double=Get Alpha_IsMinugunSelectedFloat() -> (Alpha_IsMinugunSelectedFloat:double)
    MinigunNotCrouch:double=Get Is MinigunAimLowerNotCrouch() -> (Is MinigunAimLowerNotCrouch:double)
    MinigunSelected Aim Lower In Crouch:double=Get Alpha_IsMinugunAimLowerInCrouch() -> (Alpha_IsMinugunAimLowerInCrouch:double)
   ) -> (Pose) {
    alphaBoolBlend {blendOption=Linear}
    controlRigAssetReference {blueprintRigClass=&/M_MED_RogueNinja_ControlRig.M_MED_RogueNinja_ControlRig_C}
   }
   Save cached pose 'ControlRigCache'(Pose=@n58) {cacheName=ControlRigCache}
   n59 = Look At - Bone: dyn_jacket_l_A1(ComponentPose=Local To Component(LocalPose=Use cached pose 'ControlRigCache'()), Alpha=1) {
    alphaBoolBlend {blendOption=Linear}
    boneToModify {boneName=dyn_jacket_l_A1}
    lookAtLocation=<6.6002 4.1068 -12.3922>
    lookAtTarget {boneReference {boneName=deform_knee_out_l}}
    lookAt_Axis {axis=<1 0>}
   }
   n60 = Look At - Bone: dyn_jacket_l_B1(ComponentPose=@n59, Alpha=1) {
    alphaBoolBlend {blendOption=Linear}
    boneToModify {boneName=dyn_jacket_l_B1}
    lookAtLocation=<-17.5476 -6.3029 4.2338>
    lookAtTarget {boneReference {boneName=dyn_l_bagB}}
    lookAt_Axis {axis=<1 0>}
   }
   n61 = Look At - Bone: dyn_jacket_r_A1(ComponentPose=@n60, Alpha=Get Alpha_WearingCapeRBEnable() -> (Alpha_WearingCapeRBEnable:double)) {
    alphaBoolBlend {blendOption=Linear}
    boneToModify {boneName=dyn_jacket_r_A1}
    lookAtLocation=<16.5461 -19.492 6.0732>
    lookAtTarget {boneReference {boneName=deform_knee_out_r}}
    lookAt_Axis {axis=<-1 0>}
   }
   n62 = Look At - Bone: dyn_jacket_r_A1(ComponentPose=@n61, Alpha=Get Alpha_IsJump Disabled RB() -> (Alpha_IsJump Disabled RB:double)) {
    alphaBoolBlend {blendOption=Linear}
    alphaScaleBiasClamp {bInterpResult=true, interpSpeedIncreasing=2}
    boneToModify {boneName=dyn_jacket_r_A1}
    lookAtLocation=<16.5461 -8.1898 4.9155>
    lookAtTarget {boneReference {boneName=deform_knee_out_r}}
    lookAt_Axis {axis=<-1 0>}
   }
   n63 = Look At - Bone: dyn_jacket_r_B1(ComponentPose=@n62, Alpha=1) {
    alphaBoolBlend {blendOption=Linear}
    boneToModify {boneName=dyn_jacket_r_B1}
    lookAtLocation=<11.446 26.4383 26.1>
    lookAtTarget {boneReference {boneName=deform_knee_out_r}}
    lookAt_Axis {axis=<-1 0>}
   }
   n64 = Get Alpha_IsSprinting() -> (Alpha_IsSprinting:double)
   n65 = "Transform (Modify) Bone - Bone: dyn_jacket_c_D1"(ComponentPose=@n63, Alpha=@n64, Translation="0,0,0", Rotation="0,0,17", Scale="1,1,1") {
    rotationMode=BMM_Additive
    scaleMode=BMM_Additive
    translationMode=BMM_Additive
    alphaBoolBlend {blendOption=Linear}
    alphaScaleBiasClamp {bInterpResult=true, interpSpeedDecreasing=4, interpSpeedIncreasing=4}
    boneToModify {boneName=dyn_jacket_c_D1}
    rotation {roll=17}
   }
   n66 = "Transform (Modify) Bone - Bone: base_jacket_r_C"(ComponentPose=@n65, Alpha=@n64, Translation="0,0,0", Rotation="0,0,10", Scale="1,1,1") {
    rotationMode=BMM_Additive
    scaleMode=BMM_Additive
    translationMode=BMM_Additive
    alphaBoolBlend {blendOption=Linear}
    alphaScaleBiasClamp {bInterpResult=true, interpSpeedDecreasing=4, interpSpeedIncreasing=4}
    boneToModify {boneName=base_jacket_r_C}
    rotation {roll=10}
   }
   n67 = "Transform (Modify) Bone - Bone: base_jacket_l_C"(ComponentPose=@n66, Alpha=@n64, Translation="0,0,0", Rotation="0,0,10", Scale="1,1,1") {
    rotationMode=BMM_Additive
    scaleMode=BMM_Additive
    translationMode=BMM_Additive
    alphaBoolBlend {blendOption=Linear}
    alphaScaleBiasClamp {bInterpResult=true, interpSpeedDecreasing=4, interpSpeedIncreasing=4}
    boneToModify {boneName=base_jacket_l_C}
    rotation {roll=10}
   }
   n68 = "Transform (Modify) Bone - Bone: spine_05_correct_l"(ComponentPose=@n67, Alpha=@n64, Translation="-1,0,0", Rotation="0,0,0", Scale="1,1,1") {
    rotationMode=BMM_Additive
    scaleMode=BMM_Additive
    translationMode=BMM_Additive
    alphaBoolBlend {blendOption=Linear}
    alphaScaleBiasClamp {bInterpResult=true, interpSpeedDecreasing=4, interpSpeedIncreasing=4}
    boneToModify {boneName=spine_05_correct_l}
    translation {x=-1}
   }
   n69 = "Transform (Modify) Bone - Bone: clavicle_correct_r"(ComponentPose=@n68, Alpha=@n64, Translation="0,3,0", Rotation="0,0,0", Scale="1,1,1") {
    rotationMode=BMM_Additive
    scaleMode=BMM_Additive
    translationMode=BMM_Additive
    alphaBoolBlend {blendOption=Linear}
    alphaScaleBiasClamp {bInterpResult=true, interpSpeedDecreasing=4, interpSpeedIncreasing=4}
    boneToModify {boneName=clavicle_correct_r}
    translation {y=3}
   }
   n70 = Get Alpha_IsJumping() -> (Alpha_IsJumping:double)
   n71 = "Transform (Modify) Bone - Bone: base_jacket_r_A"(ComponentPose=@n69, Alpha=@n70, Translation="0,0,0", Rotation="0,10,-20", Scale="1,1,1") {
    rotationMode=BMM_Additive
    scaleMode=BMM_Additive
    translationMode=BMM_Additive
    alphaBoolBlend {blendOption=Linear}
    alphaScaleBiasClamp {bInterpResult=true, interpSpeedDecreasing=4, interpSpeedIncreasing=4}
    boneToModify {boneName=base_jacket_r_A}
    rotation {roll=-20, yaw=10}
   }
   n72 = "Transform (Modify) Bone - Bone: base_jacket_l_A"(ComponentPose=@n71, Alpha=@n70, Translation="0,0,0", Rotation="20,0,0", Scale="1,1,1") {
    rotationMode=BMM_Additive
    scaleMode=BMM_Additive
    translationMode=BMM_Additive
    alphaBoolBlend {blendOption=Linear}
    alphaScaleBiasClamp {bInterpResult=true, interpSpeedDecreasing=4, interpSpeedIncreasing=4}
    boneToModify {boneName=base_jacket_l_A}
    rotation {pitch=20}
   }
   n73 = Get Alpha_IsJumpingWearingCape RB Enable() -> (Alpha_IsJumpingWearingCape RB Enable:double)
   n74 = "Transform (Modify) Bone - Bone: dyn_jacket_c_D2"(ComponentPose=@n72, Alpha=@n73, Translation="0,0,0", Rotation="-25,0,0", Scale="1,1,1") {
    rotationMode=BMM_Additive
    rotationSpace=BCS_BoneSpace
    scaleMode=BMM_Additive
    translationMode=BMM_Additive
    alphaBoolBlend {blendOption=Linear}
    alphaScaleBiasClamp {bInterpResult=true, interpSpeedDecreasing=4, interpSpeedIncreasing=4}
    boneToModify {boneName=dyn_jacket_c_D2}
    rotation {pitch=-25}
   }
   n75 = "Transform (Modify) Bone - Bone: dyn_jacket_r_C2"(ComponentPose=@n74, Alpha=@n73, Translation="0,0,0", Rotation="-30,0,0", Scale="1,1,1") {
    rotationMode=BMM_Additive
    rotationSpace=BCS_BoneSpace
    scaleMode=BMM_Additive
    translationMode=BMM_Additive
    alphaBoolBlend {blendOption=Linear}
    alphaScaleBiasClamp {bInterpResult=true, interpSpeedDecreasing=4, interpSpeedIncreasing=4}
    boneToModify {boneName=dyn_jacket_r_C2}
    rotation {pitch=-30}
   }
   n76 = "Transform (Modify) Bone - Bone: base_jacket_c_D"(
    ComponentPose=@n75
    Alpha=Get Alpha_IsJumpingIdle_RB Enabled() -> (Alpha_IsJumpingIdle_RB Enabled:double)
    Translation="0,0,0"
    Rotation="0,0,20"
    Scale="1,1,1"
   ) {
    rotationMode=BMM_Additive
    scaleMode=BMM_Additive
    translationMode=BMM_Additive
    alphaBoolBlend {blendOption=Linear}
    alphaScaleBiasClamp {bInterpResult=true, interpSpeedDecreasing=4}
    boneToModify {boneName=base_jacket_c_D}
    rotation {roll=20}
   }
   n77 = Get Alpha_JumpMoving RB() -> (Alpha_JumpMoving RB:double)
   n78 = "Transform (Modify) Bone - Bone: base_jacket_r_A"(ComponentPose=@n76, Alpha=@n77, Translation="0,0,5", Rotation="-30,20,0", Scale="1,1,1") {
    rotationMode=BMM_Additive
    scaleMode=BMM_Additive
    translationMode=BMM_Additive
    alphaBoolBlend {blendOption=Linear}
    alphaScaleBiasClamp {bInterpResult=true, interpSpeedDecreasing=3, interpSpeedIncreasing=8}
    boneToModify {boneName=base_jacket_r_A}
    rotation {pitch=-30, yaw=20}
    translation {z=5}
   }
   n79 = "Transform (Modify) Bone - Bone: base_jacket_r_B"!(ComponentPose=@n78, Alpha:float=@n77, Translation="0,0,0", Rotation="0,10,-25", Scale="1,1,1") -> (Pose) {
    rotationMode=BMM_Additive
    scaleMode=BMM_Additive
    translationMode=BMM_Additive
    alphaBoolBlend {blendOption=Linear}
    alphaScaleBiasClamp {bInterpResult=true, interpSpeedDecreasing=3, interpSpeedIncreasing=4}
    boneToModify {boneName=base_jacket_r_B}
    rotation {roll=-25, yaw=10}
   }
   n80 = "Transform (Modify) Bone - Bone: base_jacket_r_C"(
    ComponentPose=@n79
    Alpha=Get is LeftFallingInJump RB Enabled() -> (is LeftFallingInJump RB Enabled:double)
    Translation="0,0,0"
    Rotation="-1.455,-9.327,-26.751"
    Scale="1,1,1"
   ) {
    rotationMode=BMM_Additive
    scaleMode=BMM_Additive
    translationMode=BMM_Additive
    alphaBoolBlend {blendOption=Linear}
    alphaScaleBiasClamp {bInterpResult=true, interpSpeedDecreasing=3, interpSpeedIncreasing=4}
    boneToModify {boneName=base_jacket_r_C}
    rotation=<rot -1.455 -9.327 -26.751>
   }
   n81 = "Transform (Modify) Bone - Bone: base_jacket_r_A"(
    ComponentPose=@n80
    Alpha=Get is RightFallingInJump RB Enabled() -> (is RightFallingInJump RB Enabled:double)
    Translation="0,0,0"
    Rotation="20,0,0"
    Scale="1,1,1"
   ) {
    rotationMode=BMM_Additive
    scaleMode=BMM_Additive
    translationMode=BMM_Additive
    alphaBoolBlend {blendOption=Linear}
    alphaScaleBiasClamp {bInterpResult=true, interpSpeedDecreasing=3, interpSpeedIncreasing=4}
    boneToModify {boneName=base_jacket_r_A}
    rotation {pitch=20}
   }
   n82 = "Constraint - base_jacket_l_C (RR)"(ComponentPose=@n81, Alpha=1, ConstraintWeights_0=1, ConstraintWeights_1=1) {
    alphaBoolBlend {blendOption=Linear}
    alphaScaleBiasClamp {bInterpResult=true, interpSpeedDecreasing=4, interpSpeedIncreasing=4}
    boneToModify {boneName=base_jacket_l_C}
    constraintSetup [{transformType=Rotation, targetBone {boneName=dyn_jacket_c_D1}}, {transformType=Rotation, targetBone {boneName=dyn_jacket_l_B1}}]
    constraintWeights [1,1]
   }
   n83 = "Constraint - dyn_jacket_r_C1 (RR)"(ComponentPose=@n82, Alpha=1, ConstraintWeights_0=1, ConstraintWeights_1=0.5) {
    alphaBoolBlend {blendOption=Linear}
    boneToModify {boneName=dyn_jacket_r_C1}
    constraintSetup [{transformType=Rotation, targetBone {boneName=dyn_jacket_c_D1}}, {transformType=Rotation, targetBone {boneName=dyn_jacket_r_B1}}]
    constraintWeights [1,1]
   }
   n84 = RigidBody(ComponentPose=@n83, Alpha=Get Alpha_IsWearingCapeEnable() -> (Alpha_IsWearingCapeEnable:double), OverrideWorldGravity=Get Gravity Override()) {
    bOverrideWorldGravity=true
    overridePhysicsAsset=&/M_MED_RogueNinja_PhysicsMoving2d
    simulationSpace=BaseBoneSpace
    alphaBoolBlend {blendOption=Linear}
    alphaScaleBiasClamp {bInterpResult=true}
    baseBoneRef {boneName=pelvis}
    componentAppliedLinearAccClamp=<3000 3000 3000>
    componentLinearAccScale=<0.4 0.4 0.4>
    componentLinearVelScale=<0.4 0.4 0.4>
    simSpaceSettings {maxAngularAcceleration=1000, maxAngularVelocity=3, maxLinearAcceleration=700, maxLinearVelocity=500, velocityScaleZ=0, worldAlpha=0.5}
   }
   n85 = "Transform (Modify) Bone - Bone: base_jacket_l_A"(
    ComponentPose=@n84
    Alpha=Get Alpha_IsMelee Enable RB() -> (Alpha_IsMelee Enable RB:double)
    Translation="0,0,0"
    Rotation="15,0,0"
    Scale="1,1,1"
   ) {
    rotationMode=BMM_Additive
    rotationSpace=BCS_BoneSpace
    scaleMode=BMM_Additive
    translationMode=BMM_Additive
    alphaBoolBlend {blendOption=Linear}
    alphaScaleBiasClamp {bInterpResult=true, interpSpeedDecreasing=3, interpSpeedIncreasing=4}
    boneToModify {boneName=base_jacket_l_A}
    rotation {pitch=15}
   }
   n86 = "Transform (Modify) Bone - Bone: hood_back"(ComponentPose=@n85, Alpha=Get Alpha_IsMeleeDualWeild(), Translation="2,-1,0", Rotation="0,0,0", Scale="1,1,1") {
    rotationMode=BMM_Additive
    rotationSpace=BCS_BoneSpace
    scaleMode=BMM_Additive
    translationMode=BMM_Additive
    alphaBoolBlend {blendOption=Linear}
    alphaScaleBiasClamp {bInterpResult=true, interpSpeedDecreasing=3, interpSpeedIncreasing=4}
    boneToModify {boneName=hood_back}
    translation=<2 -1>
   }
   n87 = "Transform (Modify) Bone - Bone: base_jacket_r_B"(
    ComponentPose=@n86
    Alpha=Get Alpha_IsMeleeMovingDisableRB() -> (Alpha_IsMeleeMovingDisableRB:double)
    Translation="0,0,0"
    Rotation="10,0,0"
    Scale="1,1,1"
   ) {
    rotationMode=BMM_Additive
    rotationSpace=BCS_BoneSpace
    scaleMode=BMM_Additive
    translationMode=BMM_Additive
    alphaBoolBlend {blendOption=Linear}
    alphaScaleBiasClamp {bInterpResult=true, interpSpeedDecreasing=3, interpSpeedIncreasing=4}
    boneToModify {boneName=base_jacket_r_B}
    rotation {pitch=10}
   }
   n88 = Get RightBool / RB Enable() -> (RightBool / RB Enable:double)
   n89 = "Transform (Modify) Bone - Bone: base_jacket_r_A"(ComponentPose=@n87, Alpha=@n88, Translation="0,0,0", Rotation="-20,0,0", Scale="1,1,1") {
    rotationMode=BMM_Additive
    scaleMode=BMM_Additive
    translationMode=BMM_Additive
    alphaBoolBlend {blendOption=Linear}
    alphaScaleBiasClamp {bInterpResult=true, interpSpeedDecreasing=3, interpSpeedIncreasing=4}
    boneToModify {boneName=base_jacket_r_A}
    rotation {pitch=-20}
   }
   n90 = "Transform (Modify) Bone - Bone: base_jacket_r_B"(ComponentPose=@n89, Alpha=@n88, Translation="0,0,0", Rotation="-20,0,0", Scale="1,1,1") {
    rotationMode=BMM_Additive
    scaleMode=BMM_Additive
    translationMode=BMM_Additive
    alphaBoolBlend {blendOption=Linear}
    alphaScaleBiasClamp {bInterpResult=true, interpSpeedDecreasing=3, interpSpeedIncreasing=4}
    boneToModify {boneName=base_jacket_r_B}
    rotation {pitch=-20}
   }
   n91 = Get Alpha_IsDisableRBFrontEnd Idle() -> (Alpha_IsDisableRBFrontEnd Idle:double)
   n92 = "Transform (Modify) Bone - Bone: base_jacket_r_B"(ComponentPose=@n90, Alpha=@n91, Translation="0,0,0", Rotation="-12,0,-10", Scale="1,1,1") {
    rotationMode=BMM_Additive
    scaleMode=BMM_Additive
    translationMode=BMM_Additive
    alphaBoolBlend {blendOption=Linear}
    alphaScaleBiasClamp {bInterpResult=true, interpSpeedDecreasing=3, interpSpeedIncreasing=4}
    boneToModify {boneName=base_jacket_r_B}
    rotation {pitch=-12, roll=-10}
   }
   n93 = "Transform (Modify) Bone - Bone: base_jacket_l_A"(ComponentPose=@n92, Alpha=@n91, Translation="0,0,0", Rotation="-10,0,-5", Scale="1,1,1") {
    rotationMode=BMM_Additive
    scaleMode=BMM_Additive
    translationMode=BMM_Additive
    alphaBoolBlend {blendOption=Linear}
    alphaScaleBiasClamp {bInterpResult=true, interpSpeedDecreasing=3, interpSpeedIncreasing=4}
    boneToModify {boneName=base_jacket_l_A}
    rotation {pitch=-10, roll=-5}
   }
   n94 = AND Boolean(A=Get is Wearing Cape() -> (is Wearing Cape:bool), B=NOT Boolean(A=Get bIsMoving2D()))
   n95 = "Transform (Modify) Bone - Bone: base_jacket_l_B"!(ComponentPose=@n93, bAlphaBoolEnabled:bool=@n94, Translation="-0.546,1.654,0.252", Rotation="1.549,6.070,-0.421", Scale="1,1,1") -> (Pose) {
    alphaInputType=Bool
    rotationMode=BMM_Additive
    rotationSpace=BCS_BoneSpace
    scaleMode=BMM_Additive
    translationMode=BMM_Additive
    alphaBoolBlend {blendOption=Linear}
    alphaScaleBiasClamp {bInterpResult=true, interpSpeedDecreasing=3, interpSpeedIncreasing=4}
    boneToModify {boneName=base_jacket_l_B}
    rotation=<rot 1.549 6.07 -0.421>
    translation=<-0.546 1.654 0.252>
   }
   n96 = "Transform (Modify) Bone - Bone: base_jacket_l_C"!(ComponentPose=@n95, bAlphaBoolEnabled:bool=@n94, Translation="0,-0.5,0", Rotation="-6,0,0", Scale="1,1,1") -> (Pose) {
    alphaInputType=Bool
    rotationMode=BMM_Additive
    rotationSpace=BCS_BoneSpace
    scaleMode=BMM_Additive
    translationMode=BMM_Additive
    alphaBoolBlend {blendOption=Linear}
    alphaScaleBiasClamp {bInterpResult=true, interpSpeedDecreasing=3, interpSpeedIncreasing=4}
    boneToModify {boneName=base_jacket_l_C}
    rotation {pitch=-6}
    translation {y=-0.5}
   }
   n97 = "Transform (Modify) Bone - Bone: dyn_jacket_l_C3"!(ComponentPose=@n96, bAlphaBoolEnabled:bool=@n94, Translation="0,0,0", Rotation="-6,0,0", Scale="1,1,1") -> (Pose) {
    alphaInputType=Bool
    rotationMode=BMM_Additive
    rotationSpace=BCS_BoneSpace
    scaleMode=BMM_Additive
    translationMode=BMM_Additive
    alphaBoolBlend {blendOption=Linear}
    alphaScaleBiasClamp {bInterpResult=true, interpSpeedDecreasing=3, interpSpeedIncreasing=4}
    boneToModify {boneName=dyn_jacket_l_C3}
    rotation {pitch=-6}
   }
   n98 = "Transform (Modify) Bone - Bone: base_jacket_l_C"!(ComponentPose=@n97, Alpha:float=Get Alpha_IsJumpingWearingCape RB Enable(), Translation="0,0,0", Rotation="-17,0,0", Scale="1,1,1") -> (Pose) {
    rotationMode=BMM_Additive
    rotationSpace=BCS_BoneSpace
    scaleMode=BMM_Additive
    translationMode=BMM_Additive
    alphaBoolBlend {blendOption=Linear}
    alphaScaleBiasClamp {bInterpResult=true, interpSpeedDecreasing=3, interpSpeedIncreasing=4}
    boneToModify {boneName=base_jacket_l_C}
    rotation {pitch=-17}
   }
   n99 = Get Is Moving2D RB Disable() -> (Is Moving2D RB Disable:double)
   n100 = "Transform (Modify) Bone - Bone: base_jacket_l_C"(ComponentPose=@n98, Alpha=@n99, Translation="0,0,0", Rotation="5,0,0", Scale="1,1,1") {
    rotationMode=BMM_Additive
    rotationSpace=BCS_BoneSpace
    scaleMode=BMM_Additive
    translationMode=BMM_Additive
    alphaBoolBlend {blendOption=Linear}
    alphaScaleBiasClamp {bInterpResult=true, interpSpeedDecreasing=3, interpSpeedIncreasing=4}
    boneToModify {boneName=base_jacket_l_C}
    rotation {pitch=5}
   }
   n101 = "Transform (Modify) Bone - Bone: base_jacket_r_C"(ComponentPose=@n100, Alpha=@n99, Translation="0,0,0", Rotation="5,0,0", Scale="1,1,1") {
    rotationMode=BMM_Additive
    rotationSpace=BCS_BoneSpace
    scaleMode=BMM_Additive
    translationMode=BMM_Additive
    alphaBoolBlend {blendOption=Linear}
    alphaScaleBiasClamp {bInterpResult=true, interpSpeedDecreasing=3, interpSpeedIncreasing=4}
    boneToModify {boneName=base_jacket_r_C}
    rotation {pitch=5}
   }
   n102 = "Transform (Modify) Bone - Bone: base_jacket_c_D"(ComponentPose=@n101, Alpha=@n99, Translation="0,0,0", Rotation="5,0,0", Scale="1,1,1") {
    rotationMode=BMM_Additive
    rotationSpace=BCS_BoneSpace
    scaleMode=BMM_Additive
    translationMode=BMM_Additive
    alphaBoolBlend {blendOption=Linear}
    alphaScaleBiasClamp {bInterpResult=true, interpSpeedDecreasing=3, interpSpeedIncreasing=4}
    boneToModify {boneName=base_jacket_c_D}
    rotation {pitch=5}
   }
   n103 = "Transform (Modify) Bone - Bone: base_jacket_r_B"(ComponentPose=@n102, Alpha=@n99, Translation="0,0,0", Rotation="5,0,0", Scale="1,1,1") {
    rotationMode=BMM_Additive
    rotationSpace=BCS_BoneSpace
    scaleMode=BMM_Additive
    translationMode=BMM_Additive
    alphaBoolBlend {blendOption=Linear}
    alphaScaleBiasClamp {bInterpResult=true, interpSpeedDecreasing=3, interpSpeedIncreasing=4}
    boneToModify {boneName=base_jacket_r_B}
    rotation {pitch=5}
   }
   n104 = "Transform (Modify) Bone - Bone: base_jacket_l_C"(ComponentPose=@n103, Alpha=Get Is MeleeDisableRB() -> (Is MeleeDisableRB:double), Translation="0,0,0", Rotation="3,0,0", Scale="1,1,1") {
    rotationMode=BMM_Additive
    rotationSpace=BCS_BoneSpace
    scaleMode=BMM_Additive
    translationMode=BMM_Additive
    alphaBoolBlend {blendOption=Linear}
    alphaScaleBiasClamp {bInterpResult=true, interpSpeedDecreasing=3, interpSpeedIncreasing=4}
    boneToModify {boneName=base_jacket_l_C}
    rotation {pitch=3}
   }
   Save cached pose 'Moving2DCache'(Pose=Component To Local(ComponentPose=@n104)) {cacheName=Moving2DCache}
   n105 = Look At - Bone: base_jacket_l_A(ComponentPose=Local To Component(LocalPose=Use cached pose 'ControlRigCache'()), Alpha=1) {
    alphaBoolBlend {blendOption=Linear}
    boneToModify {boneName=base_jacket_l_A}
    lookAtLocation=<3.9152 15.2808 -9.1961>
    lookAtTarget {boneReference {boneName=calf_l}}
    lookAt_Axis {axis=<1 0>}
   }
   n106 = Look At - Bone: base_jacket_l_B(ComponentPose=@n105, Alpha=1) {
    alphaBoolBlend {blendOption=Linear}
    boneToModify {boneName=base_jacket_l_B}
    lookAtLocation=<-14.4043 5.7216 -19.8196>
    lookAtTarget {boneReference {boneName=calf_l}}
    lookAt_Axis {axis=<1 0>}
   }
   n107 = "Constraint - base_jacket_l_C (P)"(ComponentPose=@n106, Alpha=1, ConstraintWeights_0=1) {
    alphaBoolBlend {blendOption=Linear}
    boneToModify {boneName=base_jacket_l_C}
    constraintSetup [{targetBone {boneName=deform_glute_l}}]
    constraintWeights [1]
   }
   n108 = Look At - Bone: base_jacket_r_A(ComponentPose=@n107, Alpha=1) {
    alphaBoolBlend {blendOption=Linear}
    boneToModify {boneName=base_jacket_r_A}
    lookAtLocation=<-3.4763 -27.1577 12.9082>
    lookAtTarget {boneReference {boneName=calf_r}}
    lookAt_Axis {axis=<-1 0>}
   }
   n109 = Look At - Bone: base_jacket_r_B(ComponentPose=@n108, Alpha=1) {
    alphaBoolBlend {blendOption=Linear}
    boneToModify {boneName=base_jacket_r_B}
    lookAtLocation=<18.8791 -7.9745 24.6407>
    lookAtTarget {boneReference {boneName=calf_r}}
    lookAt_Axis {axis=<-1 0>}
   }
   n110 = "Constraint - base_jacket_c_D (R)"(ComponentPose=@n109, Alpha=Get Alpha_isSkydiveFloating(), ConstraintWeights_0=1) {
    alphaBoolBlend {blendOption=Linear}
    boneToModify {boneName=base_jacket_c_D}
    constraintSetup [{transformType=Rotation, targetBone {boneName=pelvis}}]
    constraintWeights [1]
   }
   n111 = "Constraint - base_jacket_r_C (P)"(ComponentPose=@n110, Alpha=0, ConstraintWeights_0=1) {
    alphaBoolBlend {blendOption=Linear}
    boneToModify {boneName=base_jacket_r_C}
    constraintSetup [{targetBone {boneName=deform_glute_r}}]
    constraintWeights [1]
   }
   n112 = Get Alpha_isSkydiveModeRBDisable() -> (Alpha_isSkydiveModeRBDisable:float)
   n113 = "Transform (Modify) Bone - Bone: base_jacket_r_C"(ComponentPose=@n111, Alpha=@n112, Translation="0,0,0", Rotation="3,0,3", Scale="1,1,1") {
    rotationMode=BMM_Additive
    rotationSpace=BCS_BoneSpace
    alphaBoolBlend {blendOption=Linear}
    alphaScaleBiasClamp {bInterpResult=true, interpSpeedDecreasing=2}
    boneToModify {boneName=base_jacket_r_C}
    rotation {pitch=3, roll=3}
   }
   n114 = "Transform (Modify) Bone - Bone: base_jacket_c_D"(ComponentPose=@n113, Alpha=@n112, Translation="0,0,0", Rotation="3,0,0", Scale="1,1,1") {
    rotationMode=BMM_Additive
    rotationSpace=BCS_BoneSpace
    alphaBoolBlend {blendOption=Linear}
    alphaScaleBiasClamp {bInterpResult=true, interpSpeedDecreasing=2}
    boneToModify {boneName=base_jacket_c_D}
    rotation {pitch=3}
   }
   n115 = "Transform (Modify) Bone - Bone: base_jacket_l_C"(ComponentPose=@n114, Alpha=Get Alpha_skyDiving() -> (Alpha_skyDiving:double), Translation="0,0,0", Rotation="5,0,0", Scale="1,1,1") {
    rotationMode=BMM_Additive
    rotationSpace=BCS_BoneSpace
    alphaBoolBlend {blendOption=Linear}
    alphaScaleBiasClamp {bInterpResult=true, interpSpeedDecreasing=2, interpSpeedIncreasing=5}
    boneToModify {boneName=base_jacket_l_C}
    rotation {pitch=5}
   }
   n116 = Get Alpha_openParachute()
   n117 = "Transform (Modify) Bone - Bone: base_jacket_l_C"(ComponentPose=@n115, Alpha=@n116, Translation="0,0,0", Rotation="7,0,0", Scale="1,1,1") {
    rotationMode=BMM_Additive
    rotationSpace=BCS_BoneSpace
    alphaBoolBlend {blendOption=Linear}
    alphaScaleBiasClamp {bInterpResult=true, interpSpeedDecreasing=5, interpSpeedIncreasing=5}
    boneToModify {boneName=base_jacket_l_C}
    rotation {pitch=7}
   }
   n118 = "Transform (Modify) Bone - Bone: base_jacket_c_D"(ComponentPose=@n117, Alpha=@n116, Translation="0,0,0", Rotation="7,0,0", Scale="1,1,1") {
    rotationMode=BMM_Additive
    rotationSpace=BCS_BoneSpace
    alphaBoolBlend {blendOption=Linear}
    alphaScaleBiasClamp {bInterpResult=true, interpSpeedIncreasing=5}
    boneToModify {boneName=base_jacket_c_D}
    rotation {pitch=7}
   }
   n119 = "Transform (Modify) Bone - Bone: base_jacket_r_C"(ComponentPose=@n118, Alpha=@n116, Translation="0,0,0", Rotation="10,0,0", Scale="1,1,1") {
    rotationMode=BMM_Additive
    rotationSpace=BCS_BoneSpace
    alphaBoolBlend {blendOption=Linear}
    alphaScaleBiasClamp {bInterpResult=true, interpSpeedIncreasing=5}
    boneToModify {boneName=base_jacket_r_C}
    rotation {pitch=10}
   }
   n120 = "Transform (Modify) Bone - Bone: clavicle_correct_r"(ComponentPose=@n119, Alpha=@n116, Translation="0,0,0", Rotation="-4.481,9.470,25.013", Scale="1,1,1") {
    rotationMode=BMM_Additive
    rotationSpace=BCS_BoneSpace
    translationMode=BMM_Additive
    alphaBoolBlend {blendOption=Linear}
    alphaScaleBiasClamp {bInterpResult=true, interpSpeedIncreasing=5}
    boneToModify {boneName=clavicle_correct_r}
    rotation=<rot -4.481 9.47 25.013>
   }
   n121 = "Transform (Modify) Bone - Bone: base_jacket_c_D"!(ComponentPose=@n120, bAlphaBoolEnabled:bool=Get bRigidBodyEnabled(), Translation="0,0,0", Rotation="10,0,0", Scale="1,1,1") -> (Pose) {
    alphaInputType=Bool
    rotationMode=BMM_Additive
    rotationSpace=BCS_BoneSpace
    alphaBoolBlend {blendOption=Linear}
    alphaScaleBiasClamp {bInterpResult=true, interpSpeedIncreasing=5}
    boneToModify {boneName=base_jacket_c_D}
    rotation {pitch=10}
   }
   n122 = Get Alpha_isSkydiveFloating()
   n123 = "Transform (Modify) Bone - Bone: base_jacket_c_D"!(ComponentPose=@n121, Alpha:float=@n122, Translation="0,0,0", Rotation="5,0,0", Scale="1,1,1") -> (Pose) {
    rotationMode=BMM_Additive
    rotationSpace=BCS_BoneSpace
    alphaBoolBlend {blendOption=Linear}
    alphaScaleBiasClamp {bInterpResult=true, interpSpeedIncreasing=5}
    boneToModify {boneName=base_jacket_c_D}
    rotation {pitch=5}
   }
   n124 = "Transform (Modify) Bone - Bone: spine_03_correct_l"(ComponentPose=@n123, Alpha=@n122, Translation="0,0,3", Rotation="0,0,0", Scale="1,1,1") {
    rotationMode=BMM_Additive
    rotationSpace=BCS_BoneSpace
    translationMode=BMM_Additive
    translationSpace=BCS_BoneSpace
    alphaBoolBlend {blendOption=Linear}
    alphaScaleBiasClamp {bInterpResult=true, interpSpeedIncreasing=5}
    boneToModify {boneName=spine_03_correct_l}
    translation {z=3}
   }
   n125 = Get Alpha_openParachute()
   n126 = "Transform (Modify) Bone - Bone: spine_03_correct_l"(ComponentPose=@n124, Alpha=@n125, Translation="-17,0,0", Rotation="0,0,0", Scale="1,1,1") {
    rotationMode=BMM_Additive
    rotationSpace=BCS_BoneSpace
    translationMode=BMM_Additive
    translationSpace=BCS_BoneSpace
    alphaBoolBlend {blendOption=Linear}
    alphaScaleBiasClamp {bInterpResult=true, interpSpeedIncreasing=5}
    boneToModify {boneName=spine_03_correct_l}
    translation {x=-17}
   }
   n127 = "Transform (Modify) Bone - Bone: spine_03_correct_r"(ComponentPose=@n126, Alpha=@n125, Translation="20,0,0", Rotation="0,0,0", Scale="1,1,1") {
    rotationMode=BMM_Additive
    rotationSpace=BCS_BoneSpace
    translationMode=BMM_Additive
    translationSpace=BCS_BoneSpace
    alphaBoolBlend {blendOption=Linear}
    alphaScaleBiasClamp {bInterpResult=true, interpSpeedIncreasing=5}
    boneToModify {boneName=spine_03_correct_r}
    translation {x=20}
   }
   n128 = Get Gravity Override()
   n129 = RigidBody(ComponentPose=@n127, Alpha=Get Alpha_IsWearingCapeEnable(), OverrideWorldGravity=@n128) {
    bOverrideWorldGravity=true
    overridePhysicsAsset=&/M_MED_RogueNinja_PhysicsSkyDiving
    simulationSpace=BaseBoneSpace
    alphaBoolBlend {blendOption=Linear}
    alphaScaleBiasClamp {bInterpResult=true, interpSpeedDecreasing=5, interpSpeedIncreasing=5}
    baseBoneRef {boneName=pelvis}
    componentAppliedLinearAccClamp=<3000 3000 3000>
    componentLinearAccScale=<0.4 0.4 0.4>
    componentLinearVelScale=<0.4 0.4 0.4>
    simSpaceSettings {maxAngularAcceleration=1000, maxAngularVelocity=3, maxLinearAcceleration=700, maxLinearVelocity=500, velocityScaleZ=0, worldAlpha=0.7}
   }
   n130 = RigidBody!(ComponentPose=@n129, bAlphaBoolEnabled:bool=Get is Wearing Cape(), OverrideWorldGravity=@n128, ExternalForce) -> (Pose) {
    alphaInputType=Bool
    bOverrideWorldGravity=true
    overridePhysicsAsset=&/M_MED_RogueNinja_PhysicsSkyDivingCape
    simulationSpace=BaseBoneSpace
    alphaBoolBlend {blendOption=Linear}
    alphaScaleBiasClamp {bInterpResult=true, interpSpeedDecreasing=5, interpSpeedIncreasing=5}
    baseBoneRef {boneName=pelvis}
    componentAppliedLinearAccClamp=<3000 3000 3000>
    componentLinearAccScale=<0.4 0.4 0.4>
    componentLinearVelScale=<0.4 0.4 0.4>
    simSpaceSettings {maxAngularAcceleration=1000, maxAngularVelocity=3, maxLinearAcceleration=700, maxLinearVelocity=500, velocityScaleZ=0, worldAlpha=0.7}
   }
   n131 = "Transform (Modify) Bone - Bone: base_jacket_c_D"!(ComponentPose=@n130, bAlphaBoolEnabled:bool=Get is Wearing Cape(), Translation="0,0,0", Rotation="-10,0,0", Scale="1,1,1") -> (Pose) {
    alphaInputType=Bool
    rotationMode=BMM_Additive
    rotationSpace=BCS_BoneSpace
    alphaBoolBlend {blendOption=Linear}
    alphaScaleBiasClamp {bInterpResult=true, interpSpeedIncreasing=5}
    boneToModify {boneName=base_jacket_c_D}
    rotation {pitch=-10}
   }
   n132 = Get Alpha_openParachute()
   n133 = "Transform (Modify) Bone - Bone: dyn_l_ribbonA"(ComponentPose=@n131, Alpha=@n132, Translation="0,0,2", Rotation="0,0,0", Scale="1,1,1") {
    rotationMode=BMM_Additive
    rotationSpace=BCS_BoneSpace
    translationMode=BMM_Additive
    alphaBoolBlend {blendOption=Linear}
    alphaScaleBiasClamp {bInterpResult=true, interpSpeedIncreasing=5}
    boneToModify {boneName=dyn_l_ribbonA}
    translation {z=2}
   }
   n134 = "Transform (Modify) Bone - Bone: dyn_l_ribbonB"(ComponentPose=@n133, Alpha=@n132, Translation="0,0,0", Rotation="-15,-25,0", Scale="1,1,1") {
    rotationMode=BMM_Additive
    rotationSpace=BCS_BoneSpace
    translationMode=BMM_Additive
    alphaBoolBlend {blendOption=Linear}
    alphaScaleBiasClamp {bInterpResult=true, interpSpeedIncreasing=5}
    boneToModify {boneName=dyn_l_ribbonB}
    rotation {pitch=-15, yaw=-25}
   }
   n135 = "Transform (Modify) Bone - Bone: dyn_ribbon_base"(ComponentPose=@n134, Alpha=@n132, Translation="0,0,1.5", Rotation="0,20,-30", Scale="1,1,1") {
    rotationMode=BMM_Additive
    rotationSpace=BCS_BoneSpace
    translationMode=BMM_Additive
    alphaBoolBlend {blendOption=Linear}
    alphaScaleBiasClamp {bInterpResult=true, interpSpeedIncreasing=5}
    boneToModify {boneName=dyn_ribbon_base}
    rotation {roll=-30, yaw=20}
    translation {z=1.5}
   }
   n136 = "Transform (Modify) Bone - Bone: clavicle_correct_l"(ComponentPose=@n135, Alpha=@n132, Translation="-3.5,0,0", Rotation="0,5,0", Scale="1,1,1") {
    rotationMode=BMM_Additive
    rotationSpace=BCS_BoneSpace
    translationMode=BMM_Additive
    alphaBoolBlend {blendOption=Linear}
    alphaScaleBiasClamp {bInterpResult=true, interpSpeedIncreasing=5}
    boneToModify {boneName=clavicle_correct_l}
    rotation {yaw=5}
    translation {x=-3.5}
   }
   n137 = "Transform (Modify) Bone - Bone: base_jacket_c_D"!(
    ComponentPose=@n136
    Alpha:float=Get Alpha_isSkydiveModeCape() -> (Alpha_isSkydiveModeCape:float)
    Translation="0,0,0"
    Rotation="5,0,0"
    Scale="1,1,1"
   ) -> (Pose) {
    rotationMode=BMM_Additive
    rotationSpace=BCS_BoneSpace
    translationMode=BMM_Additive
    alphaBoolBlend {blendOption=Linear}
    alphaScaleBiasClamp {bInterpResult=true, interpSpeedIncreasing=5}
    boneToModify {boneName=base_jacket_c_D}
    rotation {pitch=5}
   }
   n138 = "Transform (Modify) Bone - Bone: base_jacket_r_C"(ComponentPose=@n137, Alpha=@n1, Translation="0,0,0", Rotation="5,0,0", Scale="1,1,1") {
    rotationMode=BMM_Additive
    rotationSpace=BCS_BoneSpace
    translationMode=BMM_Additive
    alphaBoolBlend {blendOption=Linear}
    alphaScaleBiasClamp {bInterpResult=true, interpSpeedIncreasing=5}
    boneToModify {boneName=base_jacket_r_C}
    rotation {pitch=5}
   }
   Save cached pose 'SkyDivingCache'(Pose=Component To Local(ComponentPose=@n138)) {cacheName=SkyDivingCache}
   n139 = Look At - Bone: base_jacket_l_A(ComponentPose=Local To Component(LocalPose=Use cached pose 'ControlRigCache'()), Alpha=1) {
    alphaBoolBlend {blendOption=Linear}
    boneToModify {boneName=base_jacket_l_A}
    lookAtLocation=<14.3984 15.3064 -9.0579>
    lookAtTarget {boneReference {boneName=calf_l}}
    lookAt_Axis {axis=<1 0>}
   }
   n140 = Look At - Bone: base_jacket_l_B(ComponentPose=@n139, Alpha=1) {
    alphaBoolBlend {blendOption=Linear}
    boneToModify {boneName=base_jacket_l_B}
    lookAtLocation=<-3.6838 2.6821 -27.7892>
    lookAtTarget {boneReference {boneName=calf_l}}
    lookAt_Axis {axis=<1 0>}
   }
   n141 = Look At - Bone: base_jacket_r_A(ComponentPose=@n140, Alpha=1) {
    alphaBoolBlend {blendOption=Linear}
    boneToModify {boneName=base_jacket_r_A}
    lookAtLocation=<-8.9563 -27.1577 23.3311>
    lookAtTarget {boneReference {boneName=calf_r}}
    lookAt_Axis {axis=<-1 0>}
   }
   n142 = Look At - Bone: base_jacket_r_B(ComponentPose=@n141, Alpha=1) {
    alphaBoolBlend {blendOption=Linear}
    boneToModify {boneName=base_jacket_r_B}
    lookAtLocation=<14.1535 -7.8785 26.8735>
    lookAtTarget {boneReference {boneName=calf_r}}
    lookAt_Axis {axis=<-1 0>}
   }
   n143 = Get bIsPlayingEmote()
   n144 = "Transform (Modify) Bone - Bone: base_jacket_l_C"!(ComponentPose=@n142, bAlphaBoolEnabled:bool=@n143, Translation="0,0,0", Rotation="15,0,0", Scale="1,1,1") -> (Pose) {
    alphaInputType=Bool
    rotationMode=BMM_Additive
    rotationSpace=BCS_BoneSpace
    scaleMode=BMM_Additive
    scaleSpace=BCS_BoneSpace
    translationMode=BMM_Additive
    translationSpace=BCS_BoneSpace
    alphaBoolBlend {blendOption=Linear}
    alphaScaleBiasClamp {bInterpResult=true, interpSpeedDecreasing=3, interpSpeedIncreasing=4}
    boneToModify {boneName=base_jacket_l_C}
    rotation {pitch=15}
   }
   n145 = "Transform (Modify) Bone - Bone: base_jacket_c_D"!(ComponentPose=@n144, bAlphaBoolEnabled:bool=@n143, Translation="0,0,0", Rotation="15,0,0", Scale="1,1,1") -> (Pose) {
    alphaInputType=Bool
    rotationMode=BMM_Additive
    rotationSpace=BCS_BoneSpace
    scaleMode=BMM_Additive
    scaleSpace=BCS_BoneSpace
    translationMode=BMM_Additive
    translationSpace=BCS_BoneSpace
    alphaBoolBlend {blendOption=Linear}
    alphaScaleBiasClamp {bInterpResult=true, interpSpeedDecreasing=3, interpSpeedIncreasing=4}
    boneToModify {boneName=base_jacket_c_D}
    rotation {pitch=15}
   }
   n146 = "Transform (Modify) Bone - Bone: base_jacket_r_C"!(ComponentPose=@n145, bAlphaBoolEnabled:bool=@n143, Translation="0,0,0", Rotation="15,0,0", Scale="1,1,1") -> (Pose) {
    alphaInputType=Bool
    rotationMode=BMM_Additive
    rotationSpace=BCS_BoneSpace
    scaleMode=BMM_Additive
    scaleSpace=BCS_BoneSpace
    translationMode=BMM_Additive
    translationSpace=BCS_BoneSpace
    alphaBoolBlend {blendOption=Linear}
    alphaScaleBiasClamp {bInterpResult=true, interpSpeedDecreasing=3, interpSpeedIncreasing=4}
    boneToModify {boneName=base_jacket_r_C}
    rotation {pitch=15}
   }
   n147 = Look At - Bone: base_jacket_l_A(ComponentPose=Local To Component(LocalPose=Use cached pose 'ControlRigCache'()), Alpha=1) {
    alphaBoolBlend {blendOption=Linear}
    boneToModify {boneName=base_jacket_l_A}
    lookAtLocation=<7.4079 9.0875 -19.7447>
    lookAtTarget {boneReference {boneName=calf_l}}
    lookAt_Axis {axis=<1 0>}
   }
   n148 = "Constraint - base_jacket_l_B (P)"(ComponentPose=@n147, Alpha=0.7, ConstraintWeights_0=1) {
    alphaBoolBlend {blendOption=Linear}
    boneToModify {boneName=base_jacket_l_B}
    constraintSetup [{targetBone {boneName=base_jacket_l_A}}]
    constraintWeights [1]
   }
   n149 = Look At - Bone: base_jacket_r_A(ComponentPose=@n148, Alpha=1) {
    alphaBoolBlend {blendOption=Linear}
    boneToModify {boneName=base_jacket_r_A}
    lookAtLocation=<9.3652 -16.799 68.0512>
    lookAtTarget {boneReference {boneName=calf_r}}
    lookAt_Axis {axis=<-1 0>}
   }
   n150 = Look At - Bone: base_jacket_r_B(ComponentPose=@n149, Alpha=1) {
    alphaBoolBlend {blendOption=Linear}
    boneToModify {boneName=base_jacket_r_B}
    lookAtLocation=<41.7231 3.269 27.8462>
    lookAtTarget {boneReference {boneName=calf_r}}
    lookAt_Axis {axis=<-1 0>}
   }
   n151 = RigidBody!(ComponentPose=@n150, Alpha:float, OverrideWorldGravity=Get Gravity Override(), ExternalForce) -> (Pose) {
    bOverrideWorldGravity=true
    overridePhysicsAsset=&/M_MED_RogueNinja_PhysicsEmote
    simulationSpace=BaseBoneSpace
    alphaBoolBlend {blendOption=Linear}
    alphaScaleBiasClamp {bInterpResult=true}
    baseBoneRef {boneName=pelvis}
    componentAppliedLinearAccClamp=<3000 3000 3000>
    componentLinearAccScale=<0.4 0.4 0.4>
    componentLinearVelScale=<0.4 0.4 0.4>
    simSpaceSettings {maxAngularAcceleration=1000, maxAngularVelocity=3, maxLinearAcceleration=700, maxLinearVelocity=500, velocityScaleZ=0, worldAlpha=0.5}
   }
   Save cached pose 'EmoteCache'(Pose=Blend Poses by bool(
     BlendPose_0=Component To Local(ComponentPose=@n151)
     BlendPose_1=Component To Local(ComponentPose=@n146)
     BlendTime_0=0.5
     bActiveValue=Get bRigidBodyEnabled()
    ) {blendType=Linear}) {cacheName=EmoteCache}
   n152 = Look At - Bone: base_jacket_l_A(ComponentPose=Local To Component(LocalPose=Use cached pose 'ControlRigCache'()), Alpha=1) {
    alphaBoolBlend {blendOption=Linear}
    boneToModify {boneName=base_jacket_l_A}
    lookAtLocation=<4.0614 15.3064 -9.0579>
    lookAtTarget {boneReference {boneName=calf_l}}
    lookAt_Axis {axis=<1 0>}
   }
   n153 = Look At - Bone: base_jacket_l_B(ComponentPose=@n152, Alpha=1) {
    alphaBoolBlend {blendOption=Linear}
    boneToModify {boneName=base_jacket_l_B}
    lookAtLocation=<-3.6838 2.6821 -27.7892>
    lookAtTarget {boneReference {boneName=calf_l}}
    lookAt_Axis {axis=<1 0>}
   }
   n154 = Look At - Bone: base_jacket_r_A(ComponentPose=@n153, Alpha=1) {
    alphaBoolBlend {blendOption=Linear}
    boneToModify {boneName=base_jacket_r_A}
    lookAtLocation=<-3.4763 -27.1577 12.9082>
    lookAtTarget {boneReference {boneName=calf_r}}
    lookAt_Axis {axis=<-1 0>}
   }
   n155 = Look At - Bone: base_jacket_r_B(ComponentPose=@n154, Alpha=1) {
    alphaBoolBlend {blendOption=Linear}
    boneToModify {boneName=base_jacket_r_B}
    lookAtLocation=<14.1535 -7.8785 27.8462>
    lookAtTarget {boneReference {boneName=calf_r}}
    lookAt_Axis {axis=<-1 0>}
   }
   n156 = Get Alpha_IsCrouchingIdle() -> (Alpha_IsCrouchingIdle:double)
   n157 = "Transform (Modify) Bone - Bone: base_jacket_l_A"(ComponentPose=@n155, Alpha=@n156, Translation="0,0,0", Rotation="25,-10,5", Scale="1,1,1") {
    rotationMode=BMM_Additive
    rotationSpace=BCS_BoneSpace
    translationMode=BMM_Additive
    translationSpace=BCS_BoneSpace
    alphaBoolBlend {blendOption=Linear}
    alphaScaleBiasClamp {bInterpResult=true, interpSpeedDecreasing=2, interpSpeedIncreasing=5}
    boneToModify {boneName=base_jacket_l_A}
    rotation=<rot 25 -10 5>
   }
   n158 = "Transform (Modify) Bone - Bone: base_jacket_r_A"(ComponentPose=@n157, Alpha=@n156, Translation="0,0,0", Rotation="17,0,0", Scale="1,1,1") {
    rotationMode=BMM_Additive
    rotationSpace=BCS_BoneSpace
    alphaBoolBlend {blendOption=Linear}
    alphaScaleBiasClamp {bInterpResult=true, interpSpeedDecreasing=2, interpSpeedIncreasing=5}
    boneToModify {boneName=base_jacket_r_A}
    rotation {pitch=17}
   }
   n159 = "Transform (Modify) Bone - Bone: base_jacket_r_C"!(ComponentPose=@n158, Alpha:float=Get Alpha_Crouch() -> (Alpha_Crouch:double), Translation="0,0,0", Rotation="7,0,0", Scale="1,1,1") -> (Pose) {
    rotationMode=BMM_Additive
    rotationSpace=BCS_BoneSpace
    alphaBoolBlend {blendOption=Linear}
    alphaScaleBiasClamp {bInterpResult=true, interpSpeedDecreasing=2, interpSpeedIncreasing=5}
    boneToModify {boneName=base_jacket_r_C}
    rotation {pitch=7}
   }
   n160 = Get Alpha_KneeSliding() -> (Alpha_KneeSliding:double)
   n161 = "Transform (Modify) Bone - Bone: base_jacket_r_B"(ComponentPose=@n159, Alpha=@n160, Translation="0,0,0", Rotation="40,0,-5", Scale="1,1,1") {
    rotationMode=BMM_Additive
    scaleMode=BMM_Additive
    translationMode=BMM_Additive
    alphaBoolBlend {blendOption=Linear}
    alphaScaleBiasClamp {bInterpResult=true, interpSpeedDecreasing=3, interpSpeedIncreasing=4}
    boneToModify {boneName=base_jacket_r_B}
    rotation {pitch=40, roll=-5}
   }
   n162 = "Transform (Modify) Bone - Bone: base_jacket_l_B"!(ComponentPose=@n161, Alpha:float=@n160, Translation="0,0,0", Rotation="-20,0,20", Scale="1,1,1") -> (Pose) {
    rotationMode=BMM_Additive
    scaleMode=BMM_Additive
    translationMode=BMM_Additive
    alphaBoolBlend {blendOption=Linear}
    alphaScaleBiasClamp {bInterpResult=true, interpSpeedDecreasing=3, interpSpeedIncreasing=4}
    boneToModify {boneName=base_jacket_l_B}
    rotation {pitch=-20, roll=20}
   }
   n163 = Get Alpha_isMinigunAimLower() -> (Alpha_isMinigunAimLower:float)
   n164 = "Transform (Modify) Bone - Bone: base_jacket_l_A"(ComponentPose=@n162, Alpha=@n163, Translation="0,0,15", Rotation="0,0,0", Scale="1,1,1") {
    rotationMode=BMM_Additive
    scaleMode=BMM_Additive
    translationMode=BMM_Additive
    alphaBoolBlend {blendOption=Linear}
    alphaScaleBiasClamp {bInterpResult=true, interpSpeedDecreasing=3, interpSpeedIncreasing=4}
    boneToModify {boneName=base_jacket_l_A}
    translation {z=15}
   }
   n165 = Look At - Bone: base_jacket_l_A(ComponentPose=@n164, Alpha=@n163) {
    alphaBoolBlend {blendOption=Linear}
    alphaScaleBiasClamp {bInterpResult=true, interpSpeedDecreasing=5, interpSpeedIncreasing=5}
    boneToModify {boneName=base_jacket_l_A}
    lookAtLocation=<-39.9289 10.0813 -7.43>
    lookAtTarget {boneReference {boneName=thigh_l}}
    lookAt_Axis {axis=<1 0>}
   }
   n166 = Look At - Bone: base_jacket_l_A(ComponentPose=@n165, Alpha=Get Alpha_IsMinugumAimUpper() -> (Alpha_IsMinugumAimUpper:double)) {
    alphaBoolBlend {blendOption=Linear}
    alphaScaleBiasClamp {bInterpResult=true, interpSpeedIncreasing=4}
    boneToModify {boneName=base_jacket_l_A}
    lookAtLocation=<-47.031 14.6023 -4.834>
    lookAtTarget {boneReference {boneName=thigh_l}}
    lookAt_Axis {axis=<1 0>}
   }
   n167 = "Transform (Modify) Bone - Bone: base_jacket_r_B"(ComponentPose=@n166, Alpha=Get Alpha_IsMelee() -> (Alpha_IsMelee:double), Translation="0,0,0", Rotation="12,-5,0", Scale="1,1,1") {
    rotationMode=BMM_Additive
    rotationSpace=BCS_BoneSpace
    scaleMode=BMM_Additive
    translationMode=BMM_Additive
    alphaBoolBlend {blendOption=Linear}
    alphaScaleBiasClamp {bInterpResult=true, interpSpeedDecreasing=3, interpSpeedIncreasing=4}
    boneToModify {boneName=base_jacket_r_B}
    rotation {pitch=12, yaw=-5}
   }
   n168 = "Transform (Modify) Bone - Bone: base_jacket_r_A"(ComponentPose=@n167, Alpha=Get Is Moving2D RB Disable(), Translation="0,0,0", Rotation="3,0,0", Scale="1,1,1") {
    rotationMode=BMM_Additive
    rotationSpace=BCS_BoneSpace
    scaleMode=BMM_Additive
    translationMode=BMM_Additive
    alphaBoolBlend {blendOption=Linear}
    alphaScaleBiasClamp {bInterpResult=true, interpSpeedDecreasing=3, interpSpeedIncreasing=4}
    boneToModify {boneName=base_jacket_r_A}
    rotation {pitch=3}
   }
   n169 = Component To Local(ComponentPose=@n168)
   n170 = "Transform (Modify) Bone - Bone: base_jacket_r_A"(
    ComponentPose=Local To Component(LocalPose=Use cached pose 'ControlRigCache'())
    Alpha=0.5
    Translation="0,0,0"
    Rotation="50,0,-30"
    Scale="1,1,1"
   ) {
    rotationMode=BMM_Additive
    rotationSpace=BCS_BoneSpace
    scaleMode=BMM_Additive
    translationMode=BMM_Additive
    translationSpace=BCS_BoneSpace
    alphaBoolBlend {blendOption=Linear}
    alphaScaleBiasClamp {bInterpResult=true, interpSpeedDecreasing=5, interpSpeedIncreasing=15}
    boneToModify {boneName=base_jacket_r_A}
    rotation {pitch=50, roll=-30}
   }
   n171 = "Transform (Modify) Bone - Bone: base_jacket_l_B"(ComponentPose=@n170, Alpha=1, Translation="-2.5,-2,0", Rotation="-5,20,-15", Scale="1,1,1") {
    rotationMode=BMM_Additive
    rotationSpace=BCS_BoneSpace
    translationMode=BMM_Additive
    translationSpace=BCS_BoneSpace
    alphaBoolBlend {blendOption=Linear}
    alphaScaleBiasClamp {bInterpResult=true, interpSpeedIncreasing=15}
    boneToModify {boneName=base_jacket_l_B}
    rotation=<rot -5 20 -15>
    translation=<-2.5 -2>
   }
   n172 = Get Gravity Override()
   n173 = RigidBody(ComponentPose=@n171, Alpha=Get Alpha_IsWearingCapeEnable(), OverrideWorldGravity=@n172) {
    bOverrideWorldGravity=true
    overridePhysicsAsset=&/M_MED_RogueNinja_PhysicsCrouch
    simulationSpace=BaseBoneSpace
    alphaBoolBlend {blendOption=Linear}
    alphaScaleBiasClamp {bInterpResult=true}
    baseBoneRef {boneName=pelvis}
    componentAppliedLinearAccClamp=<3000 3000 3000>
    componentLinearAccScale=<0.4 0.4 0.4>
    componentLinearVelScale=<0.4 0.4 0.4>
    simSpaceSettings {maxAngularAcceleration=1000, maxAngularVelocity=3, maxLinearAcceleration=700, maxLinearVelocity=500, velocityScaleZ=0, worldAlpha=0.5}
   }
   n174 = RigidBody!(ComponentPose=@n173, bAlphaBoolEnabled:bool=Get is Wearing Cape(), OverrideWorldGravity=@n172, ExternalForce) -> (Pose) {
    alphaInputType=Bool
    bOverrideWorldGravity=true
    overridePhysicsAsset=&/M_MED_RogueNinja_PhysicsCrouchCape
    simulationSpace=BaseBoneSpace
    alphaBoolBlend {blendOption=Linear}
    alphaScaleBiasClamp {bInterpResult=true}
    baseBoneRef {boneName=pelvis}
    componentAppliedLinearAccClamp=<3000 3000 3000>
    componentLinearAccScale=<0.4 0.4 0.4>
    componentLinearVelScale=<0.4 0.4 0.4>
    simSpaceSettings {maxAngularAcceleration=1000, maxAngularVelocity=3, maxLinearAcceleration=700, maxLinearVelocity=500, velocityScaleZ=0, worldAlpha=0.5}
   }
   n175 = "Transform (Modify) Bone - Bone: base_jacket_r_A"(ComponentPose=@n174, Alpha=Get RightCrouchMele() -> (RightCrouchMele:double), Translation="0,0,0", Rotation="15,-5,0", Scale="1,1,1") {
    rotationMode=BMM_Additive
    rotationSpace=BCS_BoneSpace
    translationMode=BMM_Additive
    translationSpace=BCS_BoneSpace
    alphaBoolBlend {blendOption=Linear}
    alphaScaleBiasClamp {bInterpResult=true, interpSpeedDecreasing=5, interpSpeedIncreasing=15}
    boneToModify {boneName=base_jacket_r_A}
    rotation {pitch=15, yaw=-5}
   }
   n176 = Get Is CrouchMoving2d() -> (Is CrouchMoving2d:double)
   n177 = "Transform (Modify) Bone - Bone: base_jacket_r_A"(ComponentPose=@n175, Alpha=@n176, Translation="0,0,0", Rotation="14.243,-10.007,-15.002", Scale="1,1,1") {
    rotationMode=BMM_Additive
    rotationSpace=BCS_BoneSpace
    translationMode=BMM_Additive
    translationSpace=BCS_BoneSpace
    alphaBoolBlend {blendOption=Linear}
    alphaScaleBiasClamp {bInterpResult=true, interpSpeedDecreasing=5, interpSpeedIncreasing=15}
    boneToModify {boneName=base_jacket_r_A}
    rotation=<rot 14.243 -10.007 -15.002>
   }
   n178 = "Transform (Modify) Bone - Bone: base_jacket_l_B"(ComponentPose=@n177, Alpha=@n176, Translation="0,0,0", Rotation="10,0,0", Scale="1,1,1") {
    rotationMode=BMM_Additive
    rotationSpace=BCS_BoneSpace
    translationMode=BMM_Additive
    translationSpace=BCS_BoneSpace
    alphaBoolBlend {blendOption=Linear}
    alphaScaleBiasClamp {bInterpResult=true, interpSpeedDecreasing=1}
    boneToModify {boneName=base_jacket_l_B}
    rotation {pitch=10}
   }
   n179 = "Transform (Modify) Bone - Bone: base_jacket_r_B"(ComponentPose=@n178, Alpha=@n176, Translation="0,0,0", Rotation="0,0,-15", Scale="1,1,1") {
    rotationMode=BMM_Additive
    rotationSpace=BCS_BoneSpace
    translationMode=BMM_Additive
    translationSpace=BCS_BoneSpace
    alphaBoolBlend {blendOption=Linear}
    alphaScaleBiasClamp {bInterpResult=true, interpSpeedDecreasing=1}
    boneToModify {boneName=base_jacket_r_B}
    rotation {roll=-15}
   }
   n180 = "Transform (Modify) Bone - Bone: base_jacket_l_A"(ComponentPose=@n179, Alpha=@n176, Translation="0,0,0", Rotation="0,0,0", Scale="1,1,1") {
    rotationMode=BMM_Additive
    rotationSpace=BCS_BoneSpace
    translationMode=BMM_Additive
    translationSpace=BCS_BoneSpace
    alphaBoolBlend {blendOption=Linear}
    alphaScaleBiasClamp {bInterpResult=true, interpSpeedDecreasing=1}
    boneToModify {boneName=base_jacket_l_A}
   }
   n181 = Get Alpha_IsCrouchingIdle()
   n182 = "Transform (Modify) Bone - Bone: dyn_jacket_l_A2"(ComponentPose=@n180, Alpha=@n181, Translation="0,0.766,-5", Rotation="-30,0,-20", Scale="1,1,1") {
    rotationMode=BMM_Additive
    rotationSpace=BCS_BoneSpace
    translationMode=BMM_Additive
    alphaBoolBlend {blendOption=Linear}
    alphaScaleBiasClamp {bInterpResult=true, interpSpeedDecreasing=2}
    boneToModify {boneName=dyn_jacket_l_A2}
    rotation {pitch=-30, roll=-20}
    translation {y=0.766, z=-5}
   }
   n183 = "Transform (Modify) Bone - Bone: dyn_jacket_l_C3"!(ComponentPose=@n182, Alpha:float=@n181, Translation="-3,0,0", Rotation="0,0,0", Scale="1,1,1") -> (Pose) {
    rotationMode=BMM_Additive
    rotationSpace=BCS_BoneSpace
    translationMode=BMM_Additive
    alphaBoolBlend {blendOption=Linear}
    alphaScaleBiasClamp {bInterpResult=true, interpSpeedDecreasing=2}
    boneToModify {boneName=dyn_jacket_l_C3}
    translation {x=-3}
   }
   n184 = "Transform (Modify) Bone - Bone: dyn_jacket_l_A3"(ComponentPose=@n183, Alpha=@n181, Translation="0,0,0", Rotation="-30,0,-10", Scale="1,1,1") {
    rotationMode=BMM_Additive
    rotationSpace=BCS_BoneSpace
    translationMode=BMM_Additive
    alphaBoolBlend {blendOption=Linear}
    alphaScaleBiasClamp {bInterpResult=true, interpSpeedDecreasing=2}
    boneToModify {boneName=dyn_jacket_l_A3}
    rotation {pitch=-30, roll=-10}
   }
   n185 = "Transform (Modify) Bone - Bone: base_jacket_l_A"(ComponentPose=@n184, Alpha=@n181, Translation="2,0,4", Rotation="25,4,0", Scale="1,1,1") {
    rotationMode=BMM_Additive
    rotationSpace=BCS_BoneSpace
    translationMode=BMM_Additive
    alphaBoolBlend {blendOption=Linear}
    alphaScaleBiasClamp {bInterpResult=true, interpSpeedDecreasing=2}
    boneToModify {boneName=base_jacket_l_A}
    rotation {pitch=25, yaw=4}
    translation {x=2, z=4}
   }
   n186 = "Transform (Modify) Bone - Bone: dyn_jacket_l_B2"(ComponentPose=@n185, Alpha=@n181, Translation="0,0,0", Rotation="-10,0,-5", Scale="1,1,1") {
    rotationMode=BMM_Additive
    rotationSpace=BCS_BoneSpace
    translationMode=BMM_Additive
    alphaBoolBlend {blendOption=Linear}
    alphaScaleBiasClamp {bInterpResult=true, interpSpeedDecreasing=2}
    boneToModify {boneName=dyn_jacket_l_B2}
    rotation {pitch=-10, roll=-5}
   }
   n187 = Get Alpha_IsWearingCapeCrouch() -> (Alpha_IsWearingCapeCrouch:double)
   n188 = "Transform (Modify) Bone - Bone: base_jacket_r_B"(ComponentPose=@n186, Alpha=@n187, Translation="0,0,0", Rotation="-20,0,0", Scale="1,1,1") {
    rotationMode=BMM_Additive
    rotationSpace=BCS_BoneSpace
    translationMode=BMM_Additive
    alphaBoolBlend {blendOption=Linear}
    alphaScaleBiasClamp {bInterpResult=true, interpSpeedDecreasing=2, interpSpeedIncreasing=5}
    boneToModify {boneName=base_jacket_r_B}
    rotation {pitch=-20}
   }
   n189 = "Transform (Modify) Bone - Bone: dyn_jacket_r_B2"(ComponentPose=@n188, Alpha=@n187, Translation="0,0,0", Rotation="10,0,-20", Scale="1,1,1") {
    rotationMode=BMM_Additive
    rotationSpace=BCS_BoneSpace
    translationMode=BMM_Additive
    alphaBoolBlend {blendOption=Linear}
    alphaScaleBiasClamp {bInterpResult=true, interpSpeedDecreasing=2, interpSpeedIncreasing=5}
    boneToModify {boneName=dyn_jacket_r_B2}
    rotation {pitch=10, roll=-20}
   }
   n190 = "Transform (Modify) Bone - Bone: base_jacket_l_A"(
    ComponentPose=@n189
    Alpha=Get Alpha_IsRocketLauncherCrouchAimUP() -> (Alpha_IsRocketLauncherCrouchAimUP:double)
    Translation="-4.2,0,0"
    Rotation="-5,0,-3"
    Scale="1,1,1"
   ) {
    rotationMode=BMM_Additive
    rotationSpace=BCS_BoneSpace
    translationMode=BMM_Additive
    translationSpace=BCS_BoneSpace
    alphaBoolBlend {blendOption=Linear}
    alphaScaleBiasClamp {bInterpResult=true, interpSpeedDecreasing=2, interpSpeedIncreasing=5}
    boneToModify {boneName=base_jacket_l_A}
    rotation {pitch=-5, roll=-3}
    translation {x=-4.2}
   }
   n191 = "Transform (Modify) Bone - Bone: base_jacket_r_B"(ComponentPose=@n190, Alpha=Get Alpha_IsWearingCapeCrouch(), Translation="0,0,0", Rotation="10,0,10", Scale="1,1,1") {
    rotationMode=BMM_Additive
    rotationSpace=BCS_BoneSpace
    translationMode=BMM_Additive
    translationSpace=BCS_BoneSpace
    alphaBoolBlend {blendOption=Linear}
    alphaScaleBiasClamp {bInterpResult=true, interpSpeedDecreasing=2, interpSpeedIncreasing=5}
    boneToModify {boneName=base_jacket_r_B}
    rotation {pitch=10, roll=10}
   }
   n192 = "Transform (Modify) Bone - Bone: base_jacket_l_C"!(ComponentPose=@n191, Alpha:float=Get Alpha_IsMelee(), Translation="0,0,0", Rotation="5,0,0", Scale="1,1,1") -> (Pose) {
    rotationMode=BMM_Additive
    rotationSpace=BCS_BoneSpace
    translationMode=BMM_Additive
    translationSpace=BCS_BoneSpace
    alphaBoolBlend {blendOption=Linear}
    alphaScaleBiasClamp {bInterpResult=true, interpSpeedDecreasing=2, interpSpeedIncreasing=5}
    boneToModify {boneName=base_jacket_l_C}
    rotation {pitch=5}
   }
   n193 = Component To Local(ComponentPose=@n192)
   n194 = Get bRigidBodyEnabled()
   n195 = "Transform (Modify) Bone - Bone: spine_05_correct_l"(
    ComponentPose=Local To Component(LocalPose=Blend Poses by bool(BlendPose_0=@n193, BlendPose_1=@n169, BlendTime_0=0.5, bActiveValue=@n194) {blendType=Linear})
    Alpha=Get Alpha_Crouch()
    Translation="0,0,1.2"
    Rotation="0,0,0"
    Scale="1,1,1"
   ) {
    rotationMode=BMM_Additive
    rotationSpace=BCS_BoneSpace
    translationMode=BMM_Additive
    translationSpace=BCS_BoneSpace
    alphaBoolBlend {blendOption=Linear}
    alphaScaleBiasClamp {bInterpResult=true, interpSpeedDecreasing=2, interpSpeedIncreasing=5}
    boneToModify {boneName=spine_05_correct_l}
    translation {z=1.2}
   }
   Save cached pose 'CrouchCache'(Pose=Component To Local(ComponentPose=@n195)) {cacheName=CrouchCache}
   n196 = Look At - Bone: base_jacket_l_A(ComponentPose=Local To Component(LocalPose=Use cached pose 'ControlRigCache'()), Alpha=1) {
    alphaBoolBlend {blendOption=Linear}
    boneToModify {boneName=base_jacket_l_A}
    lookAtLocation=<16.7994 17.1413 -10.952>
    lookAtTarget {boneReference {boneName=calf_l}}
    lookAt_Axis {axis=<1 0>}
   }
   n197 = Look At - Bone: base_jacket_l_B(ComponentPose=@n196, Alpha=1) {
    alphaBoolBlend {blendOption=Linear}
    boneToModify {boneName=base_jacket_l_B}
    lookAtLocation=<-39.8014 0.5303 -18.8963>
    lookAtTarget {boneReference {boneName=calf_l}}
    lookAt_Axis {axis=<1 0>}
   }
   n198 = Look At - Bone: base_jacket_l_C(ComponentPose=@n197, Alpha=1) {
    alphaBoolBlend {blendOption=Linear}
    boneToModify {boneName=base_jacket_l_C}
    lookAtLocation=<20.1823 -29.6744 -14.4347>
    lookAtTarget {boneReference {boneName=thigh_l}}
    lookAt_Axis {axis=<1 0>}
   }
   n199 = Look At - Bone: base_jacket_c_D(ComponentPose=@n198, Alpha=1) {
    alphaBoolBlend {blendOption=Linear}
    boneToModify {boneName=base_jacket_c_D}
    lookAtLocation=<-35.7527 22.3012 18.6938>
    lookAtTarget {boneReference {boneName=thigh_r}}
    lookAt_Axis {axis=<1 0>}
   }
   n200 = Look At - Bone: base_jacket_r_A(ComponentPose=@n199, Alpha=1) {
    alphaBoolBlend {blendOption=Linear}
    boneToModify {boneName=base_jacket_r_A}
    lookAtLocation=<-3.4763 -27.1577 12.9082>
    lookAtTarget {boneReference {boneName=calf_r}}
    lookAt_Axis {axis=<-1 0>}
   }
   n201 = Look At - Bone: base_jacket_r_B(ComponentPose=@n200, Alpha=1) {
    alphaBoolBlend {blendOption=Linear}
    boneToModify {boneName=base_jacket_r_B}
    lookAtLocation=<24.6179 -13.9495 34.2583>
    lookAtTarget {boneReference {boneName=calf_r}}
    lookAt_Axis {axis=<-1 0>}
   }
   n202 = Look At - Bone: base_jacket_r_C(ComponentPose=@n201, Alpha=1) {
    alphaBoolBlend {blendOption=Linear}
    boneToModify {boneName=base_jacket_r_C}
    lookAtLocation=<6.7021 21.8809 23.1261>
    lookAtTarget {boneReference {boneName=thigh_r}}
    lookAt_Axis {axis=<-1 0>}
   }
   n203 = "Transform (Modify) Bone - Bone: base_jacket_c_D"!(ComponentPose=@n202, Alpha:float=1, Translation="10,0,4", Rotation="115,0,0", Scale="1,1,1") -> (Pose) {
    rotationMode=BMM_Additive
    rotationSpace=BCS_BoneSpace
    scaleMode=BMM_Additive
    scaleSpace=BCS_BoneSpace
    translationMode=BMM_Additive
    translationSpace=BCS_BoneSpace
    alphaBoolBlend {blendOption=Linear}
    alphaScaleBiasClamp {bInterpResult=true, interpSpeedDecreasing=2}
    boneToModify {boneName=base_jacket_c_D}
    rotation {pitch=115}
    translation {x=10, z=4}
   }
   n204 = "Transform (Modify) Bone - Bone: base_jacket_r_C"(ComponentPose=@n203, Alpha=1, Translation="-6,-2.455,3", Rotation="0,0,0", Scale="1,1,1") {
    rotationMode=BMM_Additive
    rotationSpace=BCS_BoneSpace
    scaleMode=BMM_Additive
    scaleSpace=BCS_BoneSpace
    translationMode=BMM_Additive
    translationSpace=BCS_BoneSpace
    alphaBoolBlend {blendOption=Linear}
    alphaScaleBiasClamp {bInterpResult=true, interpSpeedDecreasing=2}
    boneToModify {boneName=base_jacket_r_C}
    translation=<-6 -2.455 3>
   }
   n205 = "Transform (Modify) Bone - Bone: base_jacket_l_C"(ComponentPose=@n204, Alpha=1, Translation="2,0,3", Rotation="17,5,30", Scale="1,1,1") {
    rotationMode=BMM_Additive
    rotationSpace=BCS_BoneSpace
    scaleMode=BMM_Additive
    scaleSpace=BCS_BoneSpace
    translationMode=BMM_Additive
    translationSpace=BCS_BoneSpace
    alphaBoolBlend {blendOption=Linear}
    alphaScaleBiasClamp {bInterpResult=true, interpSpeedDecreasing=2}
    boneToModify {boneName=base_jacket_l_C}
    rotation=<rot 17 5 30>
    translation {x=2, z=3}
   }
   n206 = "Transform (Modify) Bone - Bone: base_jacket_c_D"(
    ComponentPose=@n205
    Alpha=Get Alpha_IsMeleeSliding() -> (Alpha_IsMeleeSliding:double)
    Translation="0,0,0"
    Rotation="-20,0,0"
    Scale="1,1,1"
   ) {
    rotationMode=BMM_Additive
    rotationSpace=BCS_BoneSpace
    scaleMode=BMM_Additive
    scaleSpace=BCS_BoneSpace
    translationMode=BMM_Additive
    translationSpace=BCS_BoneSpace
    alphaBoolBlend {blendOption=Linear}
    alphaScaleBiasClamp {bInterpResult=true, interpSpeedDecreasing=5}
    boneToModify {boneName=base_jacket_c_D}
    rotation {pitch=-20}
   }
   n207 = Get Gravity Override()
   n208 = Anim Dynamics - Bone: dyn_jacket_r_B1!(ComponentPose=@n206, Alpha=1, GravityOverride:Vector=@n207, ExternalForce="0,0,0") -> (Pose) {
    angularDampingOverride=0.8
    angularSpringConstant=15000
    bAngularSpring=true
    bOverrideAngularDamping=true
    bUseGravityOverride=true
    gravityScale=0.35
    simulationSpace=RootRelative
    alphaBoolBlend {blendOption=Linear}
    boundBone {boneName=dyn_jacket_r_B1}
    componentLinearAccScale=<1 1 1>
    componentLinearVelScale=<1 1 1>
    gravityOverride {z=-980}
    physicsBodyDefinitions [{
      boundBone {boneName=dyn_jacket_r_B1}
      boxExtents=<30 30 30>
      constraintSetup {angularLimitsMax=<0 10 0>, angularLimitsMin=<0 -20 0>, angularTarget=<1 0 0>}
      localJointOffset=<4 0 0>
     }]
   }
   n209 = Anim Dynamics - Bone: dyn_jacket_r_C1!(ComponentPose=@n208, Alpha=1, GravityOverride:Vector=@n207, ExternalForce="0,0,0") -> (Pose) {
    angularDampingOverride=0.8
    angularSpringConstant=15000
    bAngularSpring=true
    bOverrideAngularDamping=true
    bUseGravityOverride=true
    gravityScale=0.35
    simulationSpace=RootRelative
    alphaBoolBlend {blendOption=Linear}
    boundBone {boneName=dyn_jacket_r_C1}
    componentLinearAccScale=<1 1 1>
    componentLinearVelScale=<1 1 1>
    gravityOverride {z=-980}
    physicsBodyDefinitions [{
      boundBone {boneName=dyn_jacket_r_C1}
      boxExtents=<30 30 30>
      constraintSetup {angularLimitsMax=<0 10 0>, angularLimitsMin=<0 -20 0>, angularTarget=<1 0 0>}
      localJointOffset=<4 0 0>
     }]
   }
   n210 = Anim Dynamics - Bone: dyn_jacket_c_D1!(ComponentPose=@n209, Alpha=1, GravityOverride:Vector=@n207, ExternalForce="0,0,0") -> (Pose) {
    angularDampingOverride=0.8
    angularSpringConstant=15000
    bAngularSpring=true
    bOverrideAngularDamping=true
    bUseGravityOverride=true
    gravityScale=0.35
    simulationSpace=RootRelative
    alphaBoolBlend {blendOption=Linear}
    boundBone {boneName=dyn_jacket_c_D1}
    componentLinearAccScale=<1 1 1>
    componentLinearVelScale=<1 1 1>
    gravityOverride {z=-980}
    physicsBodyDefinitions [{
      boundBone {boneName=dyn_jacket_c_D1}
      boxExtents=<30 30 30>
      constraintSetup {angularLimitsMax=<0 20 0>, angularLimitsMin=<0 -20 0>, angularTarget=<1 0 0>}
      localJointOffset=<4 0 0>
     }]
   }
   n211 = Anim Dynamics - Bone: dyn_jacket_l_C1!(ComponentPose=@n210, Alpha=1, GravityOverride:Vector=@n207, ExternalForce="0,0,0") -> (Pose) {
    angularDampingOverride=0.8
    angularSpringConstant=15000
    bAngularSpring=true
    bOverrideAngularDamping=true
    bUseGravityOverride=true
    gravityScale=0.35
    simulationSpace=RootRelative
    alphaBoolBlend {blendOption=Linear}
    boundBone {boneName=dyn_jacket_l_C1}
    componentLinearAccScale=<1 1 1>
    componentLinearVelScale=<1 1 1>
    gravityOverride {z=-980}
    physicsBodyDefinitions [{
      boundBone {boneName=dyn_jacket_l_C1}
      boxExtents=<40 40 40>
      constraintSetup {angularLimitsMax=<0 10 0>, angularLimitsMin=<0 -10 0>, angularTarget=<1 0 0>}
      localJointOffset=<4 0 0>
     }]
   }
   n212 = Anim Dynamics - Bone: dyn_jacket_l_B1!(ComponentPose=@n211, Alpha=1, GravityOverride:Vector=@n207, ExternalForce="0,0,0") -> (Pose) {
    angularDampingOverride=0.8
    angularSpringConstant=15000
    bAngularSpring=true
    bOverrideAngularDamping=true
    bUseGravityOverride=true
    gravityScale=0.35
    simulationSpace=RootRelative
    alphaBoolBlend {blendOption=Linear}
    boundBone {boneName=dyn_jacket_l_B1}
    componentLinearAccScale=<1 1 1>
    componentLinearVelScale=<1 1 1>
    gravityOverride {z=-980}
    physicsBodyDefinitions [{
      boundBone {boneName=dyn_jacket_l_B1}
      boxExtents=<40 40 40>
      constraintSetup {angularLimitsMax=<0 20 0>, angularLimitsMin=<0 -20 0>, angularTarget=<1 0 0>}
      localJointOffset=<4 0 0>
     }]
   }
   Save cached pose 'SlidingCache'(Pose=Component To Local(ComponentPose=@n212)) {cacheName=SlidingCache}
   n213 = Look At - Bone: base_jacket_l_A(ComponentPose=Local To Component(LocalPose=Use cached pose 'ControlRigCache'()), Alpha=1) {
    alphaBoolBlend {blendOption=Linear}
    boneToModify {boneName=base_jacket_l_A}
    lookAtLocation=<13.8165 7.5024 -9.5522>
    lookAtTarget {boneReference {boneName=calf_l}}
    lookAt_Axis {axis=<1 0>}
   }
   n214 = Look At - Bone: base_jacket_l_B(ComponentPose=@n213, Alpha=1) {
    alphaBoolBlend {blendOption=Linear}
    boneToModify {boneName=base_jacket_l_B}
    lookAtLocation=<-39.8014 -0.9517 -43.8272>
    lookAtTarget {boneReference {boneName=calf_l}}
    lookAt_Axis {axis=<1 0>}
   }
   n215 = Look At - Bone: base_jacket_l_C(ComponentPose=@n214, Alpha=1) {
    alphaBoolBlend {blendOption=Linear}
    boneToModify {boneName=base_jacket_l_C}
    lookAtLocation=<22.7483 4.8184 -21.3076>
    lookAtTarget {boneReference {boneName=thigh_l}}
    lookAt_Axis {axis=<1 0>}
   }
   n216 = Look At - Bone: base_jacket_c_D(ComponentPose=@n215, Alpha=1) {
    alphaBoolBlend {blendOption=Linear}
    boneToModify {boneName=base_jacket_c_D}
    lookAtLocation=<56.9907 15.5165 -2.7857>
    lookAtTarget {boneReference {boneName=thigh_l}}
    lookAt_Axis {axis=<-1 0>}
   }
   n217 = Look At - Bone: base_jacket_r_A(ComponentPose=@n216, Alpha=1) {
    alphaBoolBlend {blendOption=Linear}
    boneToModify {boneName=base_jacket_r_A}
    lookAtLocation=<-38.6124 -21.8402 26.8119>
    lookAtTarget {boneReference {boneName=calf_r}}
    lookAt_Axis {axis=<-1 0>}
   }
   n218 = Look At - Bone: base_jacket_r_B(ComponentPose=@n217, Alpha=1) {
    alphaBoolBlend {blendOption=Linear}
    boneToModify {boneName=base_jacket_r_B}
    lookAtLocation=<-6.0175 48.1884 9.1397>
    lookAtTarget {boneReference {boneName=calf_r}}
    lookAt_Axis {axis=<-1 0>}
   }
   n219 = Look At - Bone: base_jacket_r_C(ComponentPose=@n218, Alpha=1) {
    alphaBoolBlend {blendOption=Linear}
    boneToModify {boneName=base_jacket_r_C}
    lookAtLocation=<-13.916 -58.1128 18.6223>
    lookAtTarget {boneReference {boneName=thigh_r}}
    lookAt_Axis {axis=<-1 0>}
   }
   n220 = "Transform (Modify) Bone - Bone: base_jacket_r_C"(ComponentPose=@n219, Alpha=1, Translation="0,0,0", Rotation="24.048,7.753,0", Scale="1,1,1") {
    rotationMode=BMM_Additive
    rotationSpace=BCS_BoneSpace
    scaleMode=BMM_Additive
    scaleSpace=BCS_BoneSpace
    translationMode=BMM_Additive
    translationSpace=BCS_BoneSpace
    alphaBoolBlend {blendOption=Linear}
    alphaScaleBiasClamp {bInterpResult=true, interpSpeedDecreasing=2}
    boneToModify {boneName=base_jacket_r_C}
    rotation {pitch=24.048, yaw=7.753}
   }
   n221 = "Transform (Modify) Bone - Bone: base_jacket_r_B"(ComponentPose=@n220, Alpha=1, Translation="0,0,0", Rotation="-0.000,9.053,-0.000", Scale="1,1,1") {
    rotationMode=BMM_Additive
    rotationSpace=BCS_BoneSpace
    scaleMode=BMM_Additive
    scaleSpace=BCS_BoneSpace
    translationMode=BMM_Additive
    translationSpace=BCS_BoneSpace
    alphaBoolBlend {blendOption=Linear}
    alphaScaleBiasClamp {bInterpResult=true, interpSpeedDecreasing=2}
    boneToModify {boneName=base_jacket_r_B}
    rotation=<rot 0 9.053 0>
   }
   n222 = "Transform (Modify) Bone - Bone: base_jacket_c_D"(ComponentPose=@n221, Alpha=1, Translation="0,0,0", Rotation="-30,-11.767,2.03", Scale="1,1,1") {
    rotationMode=BMM_Additive
    rotationSpace=BCS_BoneSpace
    scaleMode=BMM_Additive
    scaleSpace=BCS_BoneSpace
    translationMode=BMM_Additive
    translationSpace=BCS_BoneSpace
    alphaBoolBlend {blendOption=Linear}
    alphaScaleBiasClamp {bInterpResult=true, interpSpeedDecreasing=2, interpSpeedIncreasing=3}
    boneToModify {boneName=base_jacket_c_D}
    rotation=<rot -30 -11.767 2.03>
   }
   n223 = Get Gravity Override()
   n224 = Anim Dynamics - Bone: dyn_jacket_r_B1(ComponentPose=@n222, Alpha=1, GravityOverride=@n223, ExternalForce="0,0,0") {
    angularDampingOverride=0.8
    angularSpringConstant=20
    bAngularSpring=true
    bOverrideAngularDamping=true
    bUseGravityOverride=true
    gravityScale=0.35
    simulationSpace=RootRelative
    alphaBoolBlend {blendOption=Linear}
    boundBone {boneName=dyn_jacket_r_B1}
    componentLinearAccScale=<1 1.5 1>
    componentLinearVelScale=<1 1.5 1>
    gravityOverride {z=-980}
    physicsBodyDefinitions [{
      boundBone {boneName=dyn_jacket_r_B1}
      boxExtents=<30 30 30>
      constraintSetup {angularLimitsMax=<0 20 0>, angularLimitsMin=<0 -20 0>, angularTarget=<1 0 0>}
      localJointOffset=<4 0 0>
     }]
   }
   n225 = Anim Dynamics - Bone: dyn_jacket_r_C1(ComponentPose=@n224, Alpha=1, GravityOverride=@n223, ExternalForce="0,0,0") {
    angularDampingOverride=0.8
    angularSpringConstant=20
    bAngularSpring=true
    bOverrideAngularDamping=true
    bUseGravityOverride=true
    gravityScale=0.35
    simulationSpace=RootRelative
    alphaBoolBlend {blendOption=Linear}
    boundBone {boneName=dyn_jacket_r_C1}
    componentLinearAccScale=<1 1.5 1>
    componentLinearVelScale=<1 1.5 1>
    gravityOverride {z=-980}
    physicsBodyDefinitions [{
      boundBone {boneName=dyn_jacket_r_C1}
      boxExtents=<30 30 30>
      constraintSetup {angularLimitsMax=<0 20 0>, angularLimitsMin=<0 -20 0>, angularTarget=<1 0 0>}
      localJointOffset=<4 0 0>
     }]
   }
   n226 = Anim Dynamics - Bone: dyn_jacket_c_D1(ComponentPose=@n225, Alpha=1, GravityOverride=@n223, ExternalForce="0,0,0") {
    angularDampingOverride=0.8
    angularSpringConstant=20
    bAngularSpring=true
    bOverrideAngularDamping=true
    bUseGravityOverride=true
    gravityScale=0.35
    simulationSpace=RootRelative
    alphaBoolBlend {blendOption=Linear}
    boundBone {boneName=dyn_jacket_c_D1}
    componentLinearAccScale=<1 1.5 1>
    componentLinearVelScale=<1 1.5 1>
    gravityOverride {z=-980}
    physicsBodyDefinitions [{
      boundBone {boneName=dyn_jacket_c_D1}
      boxExtents=<30 30 30>
      constraintSetup {angularLimitsMax=<0 20 0>, angularLimitsMin=<0 -20 0>, angularTarget=<1 0 0>}
      localJointOffset=<4 0 0>
     }]
   }
   n227 = Anim Dynamics - Bone: dyn_jacket_l_C1(ComponentPose=@n226, Alpha=1, GravityOverride=@n223, ExternalForce="0,0,0") {
    angularDampingOverride=0.8
    angularSpringConstant=20
    bAngularSpring=true
    bOverrideAngularDamping=true
    bUseGravityOverride=true
    gravityScale=0.35
    simulationSpace=RootRelative
    alphaBoolBlend {blendOption=Linear}
    boundBone {boneName=dyn_jacket_l_C1}
    componentLinearAccScale=<1 1.5 1>
    componentLinearVelScale=<1 1.5 1>
    gravityOverride {z=-980}
    physicsBodyDefinitions [{
      boundBone {boneName=dyn_jacket_l_C1}
      boxExtents=<40 40 40>
      constraintSetup {angularLimitsMax=<0 30 0>, angularLimitsMin=<0 -30 0>, angularTarget=<1 0 0>}
      localJointOffset=<6 0 0>
     }]
   }
   n228 = Anim Dynamics - Bone: dyn_jacket_l_B1(ComponentPose=@n227, Alpha=1, GravityOverride=@n223, ExternalForce="0,0,0") {
    angularDampingOverride=0.8
    angularSpringConstant=20
    bAngularSpring=true
    bOverrideAngularDamping=true
    bUseGravityOverride=true
    gravityScale=0.35
    simulationSpace=RootRelative
    alphaBoolBlend {blendOption=Linear}
    boundBone {boneName=dyn_jacket_l_B1}
    componentLinearAccScale=<1 1.5 1>
    componentLinearVelScale=<1 1.5 1>
    gravityOverride {z=-980}
    physicsBodyDefinitions [{
      boundBone {boneName=dyn_jacket_l_B1}
      boxExtents=<40 40 40>
      constraintSetup {angularLimitsMax=<0 20 0>, angularLimitsMin=<0 -20 0>, angularTarget=<1 0 0>}
      localJointOffset=<4 0 0>
     }]
   }
   Save cached pose 'SlidingWeapons'(Pose=Component To Local(ComponentPose=@n228)) {cacheName=SlidingWeapons}
   n229 = Look At - Bone: dyn_jacket_l_A1(ComponentPose=Local To Component(LocalPose=Use cached pose 'ControlRigCache'()), Alpha=1) {
    alphaBoolBlend {blendOption=Linear}
    boneToModify {boneName=dyn_jacket_l_A1}
    lookAtLocation=<6.6002 4.1068 -12.3922>
    lookAtTarget {boneReference {boneName=deform_knee_out_l}}
    lookAt_Axis {axis=<1 0>}
   }
   n230 = Look At - Bone: dyn_jacket_l_B1(ComponentPose=@n229, Alpha=1) {
    alphaBoolBlend {blendOption=Linear}
    boneToModify {boneName=dyn_jacket_l_B1}
    lookAtLocation=<-17.5476 -6.3029 4.2338>
    lookAtTarget {boneReference {boneName=dyn_l_bagB}}
    lookAt_Axis {axis=<1 0>}
   }
   n231 = Look At - Bone: dyn_jacket_r_A1(ComponentPose=@n230, Alpha=1) {
    alphaBoolBlend {blendOption=Linear}
    boneToModify {boneName=dyn_jacket_r_A1}
    lookAtLocation=<-4.3746 -19.492 15.9383>
    lookAtTarget {boneReference {boneName=deform_knee_out_r}}
    lookAt_Axis {axis=<-1 0>}
   }
   n232 = Look At - Bone: dyn_jacket_r_B1(ComponentPose=@n231, Alpha=1) {
    alphaBoolBlend {blendOption=Linear}
    boneToModify {boneName=dyn_jacket_r_B1}
    lookAtLocation=<1.7163 20.2219 29.8146>
    lookAtTarget {boneReference {boneName=deform_knee_out_r}}
    lookAt_Axis {axis=<-1 0>}
   }
   n233 = Get OnZiplineFloat() -> (OnZiplineFloat:double)
   n234 = "Transform (Modify) Bone - Bone: base_jacket_l_C"(ComponentPose=@n232, Alpha=@n233, Translation="0,0,0", Rotation="20,0,0", Scale="1,1,1") {
    rotationMode=BMM_Additive
    rotationSpace=BCS_BoneSpace
    scaleMode=BMM_Additive
    translationMode=BMM_Additive
    translationSpace=BCS_BoneSpace
    alphaBoolBlend {blendOption=Linear}
    alphaScaleBiasClamp {bInterpResult=true, interpSpeedDecreasing=3, interpSpeedIncreasing=4}
    boneToModify {boneName=base_jacket_l_C}
    rotation {pitch=20}
   }
   n235 = "Transform (Modify) Bone - Bone: base_jacket_c_D"(ComponentPose=@n234, Alpha=@n233, Translation="0,0,0", Rotation="20,0,0", Scale="1,1,1") {
    rotationMode=BMM_Additive
    rotationSpace=BCS_BoneSpace
    scaleMode=BMM_Additive
    translationMode=BMM_Additive
    translationSpace=BCS_BoneSpace
    alphaBoolBlend {blendOption=Linear}
    alphaScaleBiasClamp {bInterpResult=true, interpSpeedDecreasing=3, interpSpeedIncreasing=4}
    boneToModify {boneName=base_jacket_c_D}
    rotation {pitch=20}
   }
   n236 = "Transform (Modify) Bone - Bone: base_jacket_r_C"(ComponentPose=@n235, Alpha=@n233, Translation="0,0,0", Rotation="20,0,0", Scale="1,1,1") {
    rotationMode=BMM_Additive
    rotationSpace=BCS_BoneSpace
    scaleMode=BMM_Additive
    translationMode=BMM_Additive
    translationSpace=BCS_BoneSpace
    alphaBoolBlend {blendOption=Linear}
    alphaScaleBiasClamp {bInterpResult=true, interpSpeedDecreasing=3, interpSpeedIncreasing=4}
    boneToModify {boneName=base_jacket_r_C}
    rotation {pitch=20}
   }
   n237 = "Transform (Modify) Bone - Bone: base_jacket_r_B"(ComponentPose=@n236, Alpha=Get Is ZeplineRBDisable() -> (Is ZeplineRBDisable:double), Translation="0,0,0", Rotation="0,20,0", Scale="1,1,1") {
    rotationMode=BMM_Additive
    rotationSpace=BCS_BoneSpace
    scaleMode=BMM_Additive
    translationMode=BMM_Additive
    translationSpace=BCS_BoneSpace
    alphaBoolBlend {blendOption=Linear}
    alphaScaleBiasClamp {bInterpResult=true, interpSpeedDecreasing=3, interpSpeedIncreasing=4}
    boneToModify {boneName=base_jacket_r_B}
    rotation {yaw=20}
   }
   n238 = RigidBody!(ComponentPose=@n237, Alpha:float, OverrideWorldGravity=Get Gravity Override(), ExternalForce) -> (Pose) {
    bOverrideWorldGravity=true
    overridePhysicsAsset=&M_MED_RogueNinja_PhysicsMoving2d
    simulationSpace=BaseBoneSpace
    alphaBoolBlend {blendOption=Linear}
    alphaScaleBiasClamp {bInterpResult=true}
    baseBoneRef {boneName=pelvis}
    componentAppliedLinearAccClamp=<3000 3000 3000>
    componentLinearAccScale=<0.3 0.3 0.3>
    componentLinearVelScale=<0.3 0.3 0.3>
    simSpaceSettings {maxAngularAcceleration=1000, maxAngularVelocity=1000, maxLinearAcceleration=1000, maxLinearVelocity=1000}
   }
   Save cached pose 'ZiplineCache'(Pose=Component To Local(ComponentPose=@n238)) {cacheName=ZiplineCache}
   n239 = Look At - Bone: base_jacket_l_A(ComponentPose=Local To Component(LocalPose=Use cached pose 'ControlRigCache'()), Alpha=1) {
    alphaBoolBlend {blendOption=Linear}
    boneToModify {boneName=base_jacket_l_A}
    lookAtLocation=<25.3956 10.6898 -16.6688>
    lookAtTarget {boneReference {boneName=calf_l}}
    lookAt_Axis {axis=<1 0>}
   }
   n240 = Look At - Bone: base_jacket_l_B(ComponentPose=@n239, Alpha=1) {
    alphaBoolBlend {blendOption=Linear}
    boneToModify {boneName=base_jacket_l_B}
    lookAtLocation=<-10.8885 5.7216 -23.0666>
    lookAtTarget {boneReference {boneName=calf_l}}
    lookAt_Axis {axis=<1 0>}
   }
   n241 = "Constraint - base_jacket_l_C (P)"(ComponentPose=@n240, Alpha=1, ConstraintWeights_0=1) {
    alphaBoolBlend {blendOption=Linear}
    boneToModify {boneName=base_jacket_l_C}
    constraintSetup [{targetBone {boneName=deform_glute_l}}]
    constraintWeights [1]
   }
   n242 = Look At - Bone: base_jacket_r_A(ComponentPose=@n241, Alpha=1) {
    alphaBoolBlend {blendOption=Linear}
    boneToModify {boneName=base_jacket_r_A}
    lookAtLocation=<-3.4763 -27.1577 13.6741>
    lookAtTarget {boneReference {boneName=calf_r}}
    lookAt_Axis {axis=<-1 0>}
   }
   n243 = Look At - Bone: base_jacket_r_B(ComponentPose=@n242, Alpha=1) {
    alphaBoolBlend {blendOption=Linear}
    boneToModify {boneName=base_jacket_r_B}
    lookAtLocation=<18.8791 -7.9745 24.6407>
    lookAtTarget {boneReference {boneName=calf_r}}
    lookAt_Axis {axis=<-1 0>}
   }
   n244 = "Constraint - base_jacket_r_C (P)"(ComponentPose=@n243, Alpha=1, ConstraintWeights_0=1) {
    alphaBoolBlend {blendOption=Linear}
    boneToModify {boneName=base_jacket_r_C}
    constraintSetup [{targetBone {boneName=deform_glute_r}}]
    constraintWeights [1]
   }
   n245 = Get Alpha_isRotateLegL_OnRail() -> (Alpha_isRotateLegL_OnRail:float)
   n246 = "Transform (Modify) Bone - Bone: base_jacket_l_A"(ComponentPose=@n244, Alpha=@n245, Translation="0,-12,0", Rotation="10,15,0", Scale="1,1,1") {
    rotationMode=BMM_Additive
    rotationSpace=BCS_BoneSpace
    translationMode=BMM_Additive
    translationSpace=BCS_BoneSpace
    alphaBoolBlend {blendOption=Linear}
    alphaScaleBiasClamp {bInterpResult=true, interpSpeedDecreasing=5, interpSpeedIncreasing=5}
    boneToModify {boneName=base_jacket_l_A}
    rotation {pitch=10, yaw=15}
    translation {y=-12}
   }
   n247 = "Transform (Modify) Bone - Bone: base_jacket_l_B"(ComponentPose=@n246, Alpha=@n245, Translation="0,-12,0", Rotation="0,0,0", Scale="1,1,1") {
    rotationMode=BMM_Additive
    rotationSpace=BCS_BoneSpace
    translationMode=BMM_Additive
    translationSpace=BCS_BoneSpace
    alphaBoolBlend {blendOption=Linear}
    alphaScaleBiasClamp {bInterpResult=true, interpSpeedDecreasing=5, interpSpeedIncreasing=5}
    boneToModify {boneName=base_jacket_l_B}
    translation {y=-12}
   }
   n248 = RigidBody(ComponentPose=@n247, OverrideWorldGravity=Get Gravity Override()) {
    bOverrideWorldGravity=true
    overridePhysicsAsset=&/M_MED_RogueNinja_PhysicsRail
    simulationSpace=BaseBoneSpace
    alphaBoolBlend {blendOption=Linear}
    alphaScaleBiasClamp {bInterpResult=true}
    baseBoneRef {boneName=pelvis}
    componentAppliedLinearAccClamp=<3000 3000 3000>
    componentLinearAccScale=<0.4 0.4 0.4>
    componentLinearVelScale=<0.4 0.4 0.4>
    simSpaceSettings {maxAngularAcceleration=1000, maxAngularVelocity=3, maxLinearAcceleration=700, maxLinearVelocity=500, velocityScaleZ=0, worldAlpha=0.5}
   }
   Save cached pose 'OnRailCache'(Pose=Component To Local(ComponentPose=@n248)) {cacheName=OnRailCache}
  ]}
 EventGraph {nodes [
   Event BlueprintInitializeAnimation()
   > Set SkelMesh(SkelMesh=GetOwnerSkeletalMeshComponent())
   n1 = TryGetPawnOwner()
   > Is Valid(InputObject=@n1)
   > n2 = Cast To FortPlayerPawnAthena(Object=@n1)
   > n3 = Set PlayerPawn(PlayerPawn:Fort Player Pawn Athena Object Reference=@n2) -> (Output_Get:Fort Player Pawn Athena Object Reference)
   > Set Body_Mesh(Body_Mesh=GetSkeletalMeshForPartType(self=@n3))
   n4 = Event BlueprintUpdateAnimation()
   > n5 = Sequence!(execute) -> (then_0, then_1, then_2:Exec, then_3:Exec, then_4:Exec, then_5:Exec, then_6:Exec, then_7:Exec, then_8:Exec, then_9:Exec, then_10:Exec)
   > Sequence!(execute) -> (then_0, then_1)
   > Is Valid(InputObject=Get PlayerPawn() -> (PlayerPawn:Fort Player Pawn Athena Object Reference))
   > n6 = Branch(Condition=Get bIsInFrontEnd())
   Set WPO Globals Function(
    execute=@n6.else
    Material=Get Mat Instance Dynamic()
    Overlay Material=Get Overlay Mat Instance Dynamic()
    Skeletal Mesh=Get SkelMesh()
    "Local Noise Transform  Location"=Get Local Noise Transform Location() -> (Local Noise Transform Location:Vector)
    Local Noise Offset=Get Local Noise Offset()
   )
   > Set WPO Regional Function!(
    execute
    Material:Material Instance Dynamic Object Reference=Get Mat Instance Dynamic()
    Overlay Material:Material Instance Dynamic Object Reference=Get Overlay Mat Instance Dynamic()
    Skeletal Mesh:Skeletal Mesh Component Object Reference=Get SkelMesh()
    Velocity Transform:Name=thigh_l
    Velocity Interp Speed:double=Get Velocity Interp Speed()
    Collision Transform:Name=thigh_l
    Collision Transform Axis:EAxis Enum
    Invert Collision Transform Axis:bool=false
    Delta Time X:double=@n4.DeltaTimeX
    Previous Velocity Transform Location:"Vector (by ref)"=Get R1 Velocity Transform Location()
    Previous Velocity Transform Velocity:"Vector (by ref)"=Get R1 Velocity Transform Velocity()
    Region:int=1
   ) -> (then)
   n7 = Get bIsParachuteOpened()
   n8 = Get bIsSkydiveFloating()
   > n9 = Branch(Condition=AND Boolean(A=OR Boolean(A=@n8, B=@n7), B=Get bRigidBodyEnabled()))
   > Set WPO Pho Wind Intensity(WPO Pho Wind Intensity=0.7)
   > n10 = Set WPO Pho Turbulence(WPO Pho Turbulence:double=1) -> (Output_Get:double)
   SetScalarParameterValue(execute=[@n11.then, @n10.then], self=Get Mat Instance Dynamic(), ParameterName=ClothSim_WindStrength, Value=Get WPO Pho Wind Intensity())
   > SetScalarParameterValue(self=Get Mat Instance Dynamic(), ParameterName=ClothSim_R1_TurbulenceStrength, Value=Get WPO Pho Turbulence() -> (WPO Pho Turbulence:double))
   Set WPO Pho Wind Intensity(execute=@n9.else, WPO Pho Wind Intensity=0)
   > n11 = Set WPO Pho Turbulence(WPO Pho Turbulence=0)
   n12 = Sequence!(execute=@n5.then_9) -> (
    then_0
    then_1
    then_2:Exec
    then_3:Exec
    then_4:Exec
    then_5:Exec
    then_6:Exec
    then_7:Exec
    then_8:Exec
    then_9:Exec
    then_10:Exec
    then_11:Exec
    then_12:Exec
    then_13:Exec
    then_14:Exec
    then_15:Exec
    then_16:Exec
    then_17:Exec
    then_18:Exec
    then_19:Exec
    then_20:Exec
    then_21:Exec
    then_22:Exec
    then_23:Exec
    then_24:Exec
    then_25:Exec
    then_26:Exec
    then_27:Exec
    then_28:Exec
    then_29:Exec
    then_30:Exec
    then_31:Exec
    then_32:Exec
    then_33:Exec
    then_34:Exec
    then_35:Exec
    then_36:Exec
    then_37:Exec
    then_38:Exec
    then_39:Exec
    then_40:Exec
    then_41:Exec
   )
   n13 = Branch(execute=@n12.then_7, Condition=Get bIsKneeSliding())
   > Set Alpha_KneeSliding(Alpha_KneeSliding:double=1) -> (Output_Get:double)
   Set Alpha_KneeSliding(execute=@n13.else, Alpha_KneeSliding=0)
   n14 = Get bIsKneeSliding()
   n15 = Branch(execute=@n12.then_8, Condition=AND Boolean(A=Get bIsSliding(), B=NOT Boolean(A=@n14)))
   > Set Alpha_SlidingWeapon(Alpha_SlidingWeapon:bool=true) -> (Output_Get:bool)
   Set Alpha_SlidingWeapon(execute=@n15.else, Alpha_SlidingWeapon=false)
   n16 = Get is Wearing Cape()
   n17 = Branch(
    execute=@n12.then_14
    Condition=AND Boolean!(A=Get bIsPlayingMeleeAnim(), B=Get bRigidBodyEnabled(), C:bool=NOT Boolean(A=@n16)) -> (ReturnValue)
   )
   > Set Alpha_IsMelee Enable RB(Alpha_IsMelee Enable RB:double=1) -> (Output_Get:double)
   Set Alpha_IsMelee Enable RB(execute=@n17.else, Alpha_IsMelee Enable RB=0)
   n18 = Get bRigidBodyEnabled()
   n19 = Branch(execute=@n12.then_9, Condition=AND Boolean!(A=Get bIsSkydiveDiveMode(), B=NOT Boolean(A=@n18)) -> (ReturnValue))
   > Set Alpha_isSkydiveModeRBDisable(Alpha_isSkydiveModeRBDisable:float=1) -> (Output_Get:float)
   Set Alpha_isSkydiveModeRBDisable(execute=@n19.else, Alpha_isSkydiveModeRBDisable=0)
   n20 = Branch(execute=@n12.then_10, Condition=Get bIsSkydiveFloating())
   > Set Alpha_isSkydiveFloating(Alpha_isSkydiveFloating:float=1) -> (Output_Get:float)
   Set Alpha_isSkydiveFloating(execute=@n20.else, Alpha_isSkydiveFloating=0)
   n21 = Branch(execute=@n12.then_15, Condition=AND Boolean(A=Get bIsDBNO(), B=Get bIsMoving2D()))
   > Set Alpha_IsDBNOMoving(Alpha_IsDBNOMoving:double=1) -> (Output_Get:double)
   Set Alpha_IsDBNOMoving(execute=@n21.else, Alpha_IsDBNOMoving=0)
   n22 = Get bIsSurfaceSwimming()
   n23 = Get bIsOnGround() -> (bIsOnGround:bool)
   n24 = Get bIsJumping()
   n25 = Get bIsSkyDiving()
   n26 = Get bIsFalling()
   n27 = Get OnZipline()
   n28 = Branch(
    execute=@n12.then_11
    Condition=AND Boolean!(
     A=NOT Boolean(A=@n23)
     B=NOT Boolean(A=@n24)
     C:bool=NOT Boolean(A=@n25)
     D:bool=NOT Boolean(A=@n26)
     E:bool=NOT Boolean(A=@n22)
     F:bool=NOT Boolean(A=@n27)
    ) -> (ReturnValue)
   )
   > Set OnRailMode(OnRailMode:bool=true) -> (Output_Get:bool)
   Set OnRailMode(execute=@n28.else, OnRailMode=false)
   n29 = Branch(execute=@n12.then_17, Condition=AND Boolean!(A=Get bIsPlayingMeleeAnim(), B=Get bIsKneeSliding()) -> (ReturnValue))
   > Set Alpha_IsMeleeSliding(Alpha_IsMeleeSliding:double=1) -> (Output_Get:double)
   Set Alpha_IsMeleeSliding(execute=@n29.else, Alpha_IsMeleeSliding=0)
   n30 = Get bIsMoving2D()
   n31 = Get bRigidBodyEnabled()
   n32 = Branch(execute=@n12.then_18, Condition=AND Boolean!(A=Get bIsInFrontEnd(), B=NOT Boolean(A=@n30), C:bool=NOT Boolean(A=@n31)) -> (ReturnValue))
   > Set Alpha_IsDisableRBFrontEnd Idle(Alpha_IsDisableRBFrontEnd Idle:double=1) -> (Output_Get:double)
   Set Alpha_IsDisableRBFrontEnd Idle(execute=@n32.else, Alpha_IsDisableRBFrontEnd Idle=0)
   n33 = Branch(execute=@n12.then_32, Condition=Get Left() -> (Left:bool))
   > Set Is ParachuteIdle(Is ParachuteIdle:double=1) -> (Output_Get:double)
   Set Is ParachuteIdle(execute=@n33.else, Is ParachuteIdle=0)
   n34 = Branch(execute=@n12.then_26, Condition=AND Boolean!(A=Get bIsInFrontEnd(), B=Get bIsSkyDiving()) -> (ReturnValue))
   > Set Is SkyDivingFrontEnd(Is SkyDivingFrontEnd:double=1) -> (Output_Get:double)
   Set Is SkyDivingFrontEnd(execute=@n34.else, Is SkyDivingFrontEnd=0)
   n35 = Get bIsCrouching()
   n36 = Branch(execute=@n12.then_27, Condition=AND Boolean(A=NOT Boolean(A=@n35), B=Get AlphaIsMinigun_bool() -> (AlphaIsMinigun_bool:bool)))
   > Set Is MinigunAimLowerNotCrouch(Is MinigunAimLowerNotCrouch:double=1) -> (Output_Get:double)
   Set Is MinigunAimLowerNotCrouch(execute=@n36.else, Is MinigunAimLowerNotCrouch=0)
   n37 = Get bIsSkyDiving()
   n38 = Get bIsSurfaceSwimming()
   n39 = Branch(
    execute=@n12.then_20
    Condition=AND Boolean!(A=Get Right() -> (Right:bool), B=NOT Boolean(A=@n37), C:bool=NOT Boolean(A=@n38), D:bool=Get bIsJumping(), E:bool=Get bRigidBodyEnabled()) -> (ReturnValue)
   )
   > Set is RightFallingInJump RB Enabled(is RightFallingInJump RB Enabled:double=1) -> (Output_Get:double)
   Set is RightFallingInJump RB Enabled(execute=@n39.else, is RightFallingInJump RB Enabled=0)
   n40 = Branch(execute=@n12.then_4, Condition=Get bIsSkydiveFloating())
   > Set Alpha_skyDiving(Alpha_skyDiving:double=1) -> (Output_Get:double)
   Set Alpha_skyDiving(execute=@n40.else, Alpha_skyDiving=0)
   n41 = Get bRigidBodyEnabled()
   n42 = Branch(execute=@n12.then_12, Condition=AND Boolean!(A=Get bIsMoving2D(), B=Get bIsJumping(), C=NOT Boolean(A=@n41)) -> (ReturnValue))
   > Set Alpha_IsJump Disabled RB(Alpha_IsJump Disabled RB:double=1) -> (Output_Get:double)
   Set Alpha_IsJump Disabled RB(execute=@n42.else, Alpha_IsJump Disabled RB=0)
   n43 = Branch(execute=@n12.then_2, Condition=Get bIsJumping())
   > Set Alpha_IsJumping(Alpha_IsJumping:double=1) -> (Output_Get:double)
   Set Alpha_IsJumping(execute=@n43.else, Alpha_IsJumping=0)
   n44 = Branch(execute=@n12.then_16, Condition=AND Boolean(A=Get bIsDBNO(), B=Get bIsMoving2D(), C=Get bRigidBodyEnabled()))
   > Set Alpha_IsDBNOMovingRB Enabled(Alpha_IsDBNOMovingRB Enabled:double=1) -> (Output_Get:double)
   Set Alpha_IsDBNOMovingRB Enabled(execute=@n44.else, Alpha_IsDBNOMovingRB Enabled=0)
   n45 = Get bIsMoving2D()
   n46 = Branch(execute=@n12.then_21, Condition=AND Boolean(A=NOT Boolean(A=@n45), B=Get bIsJumping(), C=Get bRigidBodyEnabled()))
   > Set Alpha_IsJumpingIdle_RB Enabled(Alpha_IsJumpingIdle_RB Enabled:double=1) -> (Output_Get:double)
   Set Alpha_IsJumpingIdle_RB Enabled(execute=@n46.else, Alpha_IsJumpingIdle_RB Enabled=0)
   n47 = Branch(execute=@n12.then_22, Condition=AND Boolean!(A=Get is Wearing Cape(), B=Get bIsCrouching()) -> (ReturnValue))
   > Set Alpha_IsWearingCapeCrouch(Alpha_IsWearingCapeCrouch:double=1) -> (Output_Get:double)
   Set Alpha_IsWearingCapeCrouch(execute=@n47.else, Alpha_IsWearingCapeCrouch=0)
   n48 = Branch(execute=@n12.then_23, Condition=Get is Wearing Cape())
   > Set Alpha_IsWearingCapeEnable(Alpha_IsWearingCapeEnable:double=0.5) -> (Output_Get:double)
   Set Alpha_IsWearingCapeEnable(execute=@n48.else, Alpha_IsWearingCapeEnable=1)
   n49 = Get Is MeleeDualWeild_Bool() -> (Is MeleeDualWeild_Bool:bool)
   n50 = Branch(execute=@n12.then_13, Condition=AND Boolean(A=Get bIsPlayingMeleeAnim(), B=NOT Boolean(A=@n49)))
   > Set Alpha_IsMeleeDualWeild(Alpha_IsMeleeDualWeild:double=1) -> (Output_Get:double)
   Set Alpha_IsMeleeDualWeild(execute=@n50.else, Alpha_IsMeleeDualWeild=0)
   n51 = Get bIsMoving2D()
   n52 = Branch(
    execute=@n12.then_24
    Condition=AND Boolean!(A=Get bIsJumping(), B=Get bRigidBodyEnabled(), C:bool=Get is Wearing Cape(), D:bool=NOT Boolean(A=@n51)) -> (ReturnValue)
   )
   > Set Alpha_IsJumpingWearingCape RB Enable(Alpha_IsJumpingWearingCape RB Enable:double=1) -> (Output_Get:double)
   Set Alpha_IsJumpingWearingCape RB Enable(execute=@n52.else, Alpha_IsJumpingWearingCape RB Enable=0)
   n53 = Branch(execute=@n12.then_28, Condition=AND Boolean!(A=Get bIsMoving2D(), B=Get bIsCrouching()) -> (ReturnValue))
   > Set Is CrouchMoving2d(Is CrouchMoving2d:double=1) -> (Output_Get:double)
   Set Is CrouchMoving2d(execute=@n53.else, Is CrouchMoving2d=0)
   n54 = Get bIsMoving2D()
   n55 = Get bIsDBNOCarrying() -> (bIsDBNOCarrying:bool)
   n56 = Get bIsDBNOCarried()
   n57 = Branch(
    execute=@n12.then_29
    Condition=AND Boolean!(A=Get bIsDBNO(), B=NOT Boolean(A=@n54), C:bool=NOT Boolean(A=@n55), D:bool=NOT Boolean(A=@n56)) -> (ReturnValue)
   )
   > Set Is DBNOIdle(Is DBNOIdle:double=1) -> (Output_Get:double)
   Set Is DBNOIdle(execute=@n57.else, Is DBNOIdle=0)
   n58 = Branch(execute=@n12.then_25, Condition=AND Boolean!(A=Get bIsInFrontEnd(), B=Get bIsSkydiveFloating()) -> (ReturnValue))
   > Set Is FloatingFrontEnd(Is FloatingFrontEnd:double=1) -> (Output_Get:double)
   Set Is FloatingFrontEnd(execute=@n58.else, Is FloatingFrontEnd=0)
   n59 = Branch(execute=@n12.then_33, Condition=Get bIsDBNOCarried())
   > Set Alpha_IsDBNOCarried(Alpha_IsDBNOCarried:double=1) -> (Output_Get:double)
   Set Alpha_IsDBNOCarried(execute=@n59.else, Alpha_IsDBNOCarried=0)
   n60 = Branch(execute=@n12.then_30, Condition=Get bIsPlayingEmote())
   > Set Is Emote(Is Emote:double=1) -> (Output_Get:double)
   Set Is Emote(execute=@n60.else, Is Emote=0)
   n61 = Get bIsCrouching()
   n62 = Branch(execute=@n12.then_31, Condition=AND Boolean(A=Get AlphaIsRocketLauncher() -> (AlphaIsRocketLauncher:bool), B=NOT Boolean(A=@n61)))
   > Set Is RocketLauncherStand(Is RocketLauncherStand:double=1) -> (Output_Get:double)
   Set Is RocketLauncherStand(execute=@n62.else, Is RocketLauncherStand=0)
   n63 = Get is Wearing Cape()
   n64 = Branch(execute=@n12.then_1, Condition=AND Boolean(A=Get bIsSprinting(), B=NOT Boolean(A=@n63)))
   > Set Alpha_IsSprinting(Alpha_IsSprinting:double=1) -> (Output_Get:double)
   n65 = Branch(execute=@n64.else, Condition=AND Boolean(A=Get bIsSprinting(), B=Get is Wearing Cape()))
   > Set Alpha_IsSprinting(Alpha_IsSprinting=0.3)
   Set Alpha_IsSprinting(execute=@n65.else, Alpha_IsSprinting=0)
   n66 = Branch(execute=@n12.then_6, Condition=AND Boolean!(A=Get bIsJumping(), B=Get bIsMoving2D(), C:bool=Get bRigidBodyEnabled()) -> (ReturnValue))
   > Set Alpha_JumpMoving RB(Alpha_JumpMoving RB:double=1) -> (Output_Get:double)
   Set Alpha_JumpMoving RB(execute=@n66.else, Alpha_JumpMoving RB=0)
   n67 = Get is Wearing Cape()
   n68 = Branch(execute=@n12.then_3, Condition=AND Boolean!(A=Get bIsParachuteOpened(), B=NOT Boolean(A=@n67)) -> (ReturnValue))
   > Set Alpha_openParachute(Alpha_openParachute:double=1) -> (Output_Get:double)
   Set Alpha_openParachute(execute=@n68.else, Alpha_openParachute=0)
   n69 = Get bIsMoving2D()
   n70 = Branch(execute=@n12.then_0, Condition=AND Boolean(A=Get bIsCrouching(), B=NOT Boolean(A=@n69)))
   > Set Alpha_IsCrouchingIdle(Alpha_IsCrouchingIdle:double=1) -> (Output_Get:double)
   Set Alpha_IsCrouchingIdle(execute=@n70.else, Alpha_IsCrouchingIdle=0)
   n71 = Get bIsSkyDiving()
   n72 = Get bIsSurfaceSwimming()
   n73 = Branch(
    execute=@n12.then_19
    Condition=AND Boolean!(A=Get Left(), B=NOT Boolean(A=@n71), C:bool=NOT Boolean(A=@n72), D:bool=Get bIsFalling(), E:bool=Get bRigidBodyEnabled()) -> (ReturnValue)
   )
   > Set is LeftFallingInJump RB Enabled(is LeftFallingInJump RB Enabled:double=1) -> (Output_Get:double)
   Set is LeftFallingInJump RB Enabled(execute=@n73.else, is LeftFallingInJump RB Enabled=0)
   n74 = Get bRigidBodyEnabled()
   n75 = Branch(execute=@n12.then_34, Condition=AND Boolean!(A=Get bIsPlayingMeleeAnim(), B=NOT Boolean(A=@n74), C=Get bIsMoving2D()) -> (ReturnValue))
   > Set Alpha_IsMeleeMovingDisableRB(Alpha_IsMeleeMovingDisableRB:double=1) -> (Output_Get:double)
   Set Alpha_IsMeleeMovingDisableRB(execute=@n75.else, Alpha_IsMeleeMovingDisableRB=0)
   n76 = Branch(execute=@n12.then_5, Condition=Get bIsCrouching())
   > Set Alpha_Crouch(Alpha_Crouch:double=1) -> (Output_Get:double)
   Set Alpha_Crouch(execute=@n76.else, Alpha_Crouch=0)
   n77 = Branch(execute=@n12.then_35, Condition=AND Boolean(A=Get bIsMoving2D(), B=Get bRigidBodyEnabled(), C=Get is Wearing Cape()))
   > Set Alpha_WearingCapeRBEnable(Alpha_WearingCapeRBEnable:double=0.5) -> (Output_Get:double)
   n78 = Get bIsMoving2D()
   n79 = Branch(execute=@n77.else, Condition=AND Boolean(A=Get bRigidBodyEnabled(), B=Get is Wearing Cape(), C=NOT Boolean(A=@n78)))
   > Set Alpha_WearingCapeRBEnable(Alpha_WearingCapeRBEnable=0.7)
   Set Alpha_WearingCapeRBEnable(execute=@n79.else, Alpha_WearingCapeRBEnable=0.9)
   n80 = Get bRigidBodyEnabled()
   n81 = Branch(execute=@n12.then_36, Condition=AND Boolean!(A=NOT Boolean(A=@n80), B=Get bIsMoving2D()) -> (ReturnValue))
   > Set Is Moving2D RB Disable(Is Moving2D RB Disable:double=1) -> (Output_Get:double)
   Set Is Moving2D RB Disable(execute=@n81.else, Is Moving2D RB Disable=0)
   n82 = Branch(execute=@n12.then_37, Condition=Get bIsPlayingMeleeAnim())
   > Set Alpha_IsMelee(Alpha_IsMelee:double=1) -> (Output_Get:double)
   Set Alpha_IsMelee(execute=@n82.else, Alpha_IsMelee=0)
   n83 = Branch(execute=@n12.then_38, Condition=AND Boolean(A=Get is Wearing Cape(), B=Get bIsSkydiveDiveMode()))
   > Set Alpha_isSkydiveModeCape(Alpha_isSkydiveModeCape:float=1) -> (Output_Get:float)
   Set Alpha_isSkydiveModeCape(execute=@n83.else, Alpha_isSkydiveModeCape=0)
   n84 = Get bRigidBodyEnabled()
   n85 = Branch(execute=@n12.then_39, Condition=AND Boolean(A=Get bIsSkydiveFloating(), B=NOT Boolean(A=@n84)))
   > Set Alpha_isSkydiveFloating_RBDisable(Alpha_isSkydiveFloating_RBDisable:float=1) -> (Output_Get:float)
   Set Alpha_isSkydiveFloating_RBDisable(execute=@n85.else, Alpha_isSkydiveFloating_RBDisable=0)
   n86 = Branch(execute=@n12.then_40, Condition=AND Boolean(A=Get bIsMoving2D(), B=Get bIsPlayingMeleeAnim()))
   > Set Is Melee_moving2d(Is Melee_moving2d:double=0) -> (Output_Get:double)
   Set Is Melee_moving2d(execute=@n86.else, Is Melee_moving2d=1)
   n87 = Get bRigidBodyEnabled()
   n88 = Branch(execute=@n12.then_41, Condition=AND Boolean(A=Get bIsPlayingMeleeAnim(), B=NOT Boolean(A=@n87)))
   > Set Is MeleeDisableRB(Is MeleeDisableRB:double=1) -> (Output_Get:double)
   Set Is MeleeDisableRB(execute=@n88.else, Is MeleeDisableRB=0)
   n89 = PlayerMovement(execute=@n5.then_5, AnimInstance=Self-Reference(), XDotProductCoefficient=100, YDotProductCoefficient=100, ZDotProductCoefficient=100)
   n90 = BreakVector(InVec=MakeVector(X=@n89, Y=@n89.YMovement))
   n91 = "float > float"(A=@n90, B=50)
   n92 = Get bIsPlayingMeleeAnim()
   n93 = Get bIsCrouching()
   > Set RightCrouchMele(RightCrouchMele:double="To Float (Boolean)"(InBool=AND Boolean!(A=@n91, B=@n93, C:bool=@n92) -> (ReturnValue))) -> (Output_Get:double)
   n94 = Get bIsCrouching()
   > Set Is RightCrouch(Is RightCrouch:double="To Float (Boolean)"(InBool=AND Boolean!(A=@n91, B=@n94) -> (ReturnValue))) -> (Output_Get:double)
   > Set Left(Left:bool="float < float"(A=@n90, B=-50)) -> (Output_Get:bool)
   n95 = Get bRigidBodyEnabled()
   > Set RightBool / RB Enable(RightBool / RB Enable:double="To Float (Boolean)"(InBool=AND Boolean(A=@n95, B=@n91))) -> (Output_Get:double)
   > Set Right(Right:bool=@n91) -> (Output_Get:bool)
   n96 = Sequence!(execute=@n5.then_3) -> (then_0, then_1, then_2, then_3)
   n97 = Cast To FortPlayerAnimInstance(execute=@n96.then_3, Object=GetAnimInstance(self=GetOwnerSkeletalMeshComponent()))
   > Set ar_firingPosition(ar_firingPosition:int=1) -> (Output_Get:int=0)
   n98 = Get WeaponCoreAnim(self=@n97)
   > n99 = Switch on EFortWeaponCoreAnimation!(execute, Selection:EFortWeaponCoreAnimation Enum=@n98) -> (
    Melee:Exec
    Pistol:Exec
    Shotgun:Exec
    PaperBlueprint:Exec
    Rifle:Exec
    MeleeOneHand:Exec
    MachinePistol:Exec
    RocketLauncher:Exec
    GrenadeLauncher:Exec
    GoingCommando:Exec
    AssaultRifle:Exec
    TacticalShotgun:Exec
    SniperRifle:Exec
    TrapPlacement:Exec
    ShoulderLauncher:Exec
    AbilityDecoTool:Exec
    Crossbow:Exec
    C4:Exec
    RemoteControl:Exec
    DualWield:Exec
    AR_BullPup:Exec
    AR_ForwardGrip:Exec
    MedPackPaddles:Exec
    SMG_P90:Exec
    AR_DrumGun:Exec
    Consumable_Small:Exec
    Consumable_Large:Exec
    Balloon:Exec
    MountedTurret:Exec
    CreativeTool:Exec
    ExplosiveBow:Exec
    AshtonIndigo:Exec
    AshtonChicago:Exec
    MeleeDualWield:Exec
    Unarmed:Exec
   )
   Set AlphaIsMinigun_bool(
    execute=[
     @n99.AR_BullPup
     @n99.AR_DrumGun
     @n99.AR_ForwardGrip
     @n99.AbilityDecoTool
     @n99.AshtonChicago
     @n99.AshtonIndigo
     @n99.AssaultRifle
     @n99.Balloon
     @n99.C4
     @n99.Consumable_Large
     @n99.Consumable_Small
     @n99.CreativeTool
     @n99.Crossbow
     @n99.DualWield
     @n99.ExplosiveBow
     @n99.GrenadeLauncher
     @n99.MachinePistol
     @n99.MedPackPaddles
     @n99.Melee
     @n99.MeleeDualWield
     @n99.MeleeOneHand
     @n99.MountedTurret
     @n99.PaperBlueprint
     @n99.Pistol
     @n99.RemoteControl
     @n99.Rifle
     @n99.RocketLauncher
     @n99.SMG_P90
     @n99.Shotgun
     @n99.ShoulderLauncher
     @n99.SniperRifle
     @n99.TacticalShotgun
     @n99.TrapPlacement
     @n99.Unarmed
    ]
    AlphaIsMinigun_bool:bool=false
   ) -> (Output_Get:bool)
   > Set Alpha_IsMinugunSelectedFloat(Alpha_IsMinugunSelectedFloat:double=0) -> (Output_Get:double)
   Set AlphaIsMinigun_bool(execute=@n99.GoingCommando, AlphaIsMinigun_bool=true)
   > Set Alpha_IsMinugunSelectedFloat(Alpha_IsMinugunSelectedFloat=1)
   n100 = Switch on EFortWeaponCoreAnimation(execute=@n96.then_2, Selection=@n98)
   Set AlphaIsAssaultRifle_Bool(
    execute=[
     @n100.AR_BullPup
     @n100.AR_DrumGun
     @n100.AR_ForwardGrip
     @n100.AbilityDecoTool
     @n100.AshtonChicago
     @n100.AshtonIndigo
     @n100.Balloon
     @n100.C4
     @n100.Consumable_Large
     @n100.Consumable_Small
     @n100.CreativeTool
     @n100.Crossbow
     @n100.DualWield
     @n100.ExplosiveBow
     @n100.GoingCommando
     @n100.GrenadeLauncher
     @n100.MachinePistol
     @n100.MedPackPaddles
     @n100.Melee
     @n100.MeleeDualWield
     @n100.MeleeOneHand
     @n100.MountedTurret
     @n100.PaperBlueprint
     @n100.Pistol
     @n100.RemoteControl
     @n100.Rifle
     @n100.RocketLauncher
     @n100.SMG_P90
     @n100.Shotgun
     @n100.ShoulderLauncher
     @n100.SniperRifle
     @n100.TacticalShotgun
     @n100.TrapPlacement
     @n100.Unarmed
    ]
    AlphaIsAssaultRifle_Bool:bool=false
   ) -> (Output_Get:bool)
   > Set AlphaIsAssaultRifle_Float(AlphaIsAssaultRifle_Float:double=0) -> (Output_Get:double)
   Set AlphaIsAssaultRifle_Bool(execute=@n100.AssaultRifle, AlphaIsAssaultRifle_Bool=true)
   > Set AlphaIsAssaultRifle_Float(AlphaIsAssaultRifle_Float=1)
   n101 = Switch on EFortWeaponCoreAnimation(execute=@n96.then_1, Selection=@n98)
   Set is RocketLauncher_Float(
    execute=[
     @n101.AR_BullPup
     @n101.AR_DrumGun
     @n101.AR_ForwardGrip
     @n101.AbilityDecoTool
     @n101.AshtonChicago
     @n101.AshtonIndigo
     @n101.AssaultRifle
     @n101.Balloon
     @n101.C4
     @n101.Consumable_Large
     @n101.Consumable_Small
     @n101.CreativeTool
     @n101.Crossbow
     @n101.DualWield
     @n101.ExplosiveBow
     @n101.GoingCommando
     @n101.GrenadeLauncher
     @n101.MachinePistol
     @n101.MedPackPaddles
     @n101.Melee
     @n101.MeleeOneHand
     @n101.MountedTurret
     @n101.PaperBlueprint
     @n101.Pistol
     @n101.RemoteControl
     @n101.Rifle
     @n101.SMG_P90
     @n101.Shotgun
     @n101.ShoulderLauncher
     @n101.SniperRifle
     @n101.TacticalShotgun
     @n101.TrapPlacement
     @n101.Unarmed
    ]
    is RocketLauncher_Float:double=0
   ) -> (Output_Get:double)
   > Set AlphaIsRocketLauncher(AlphaIsRocketLauncher:bool=false) -> (Output_Get:bool)
   > Set Is RocketLauncherCorrect(Is RocketLauncherCorrect:double=0) -> (Output_Get:double)
   Set is RocketLauncher_Float(execute=@n101.RocketLauncher, is RocketLauncher_Float=1)
   > Set AlphaIsRocketLauncher(AlphaIsRocketLauncher=true)
   > Set Is RocketLauncherCorrect(Is RocketLauncherCorrect=1)
   n102 = Switch on EFortWeaponCoreAnimation(execute=@n96.then_0, Selection=@n98)
   Set Is DetectedDualWeild(
    execute=[
     @n102.AR_BullPup
     @n102.AR_DrumGun
     @n102.AR_ForwardGrip
     @n102.AbilityDecoTool
     @n102.AshtonChicago
     @n102.AshtonIndigo
     @n102.AssaultRifle
     @n102.Balloon
     @n102.C4
     @n102.Consumable_Large
     @n102.Consumable_Small
     @n102.CreativeTool
     @n102.Crossbow
     @n102.DualWield
     @n102.GoingCommando
     @n102.GrenadeLauncher
     @n102.MachinePistol
     @n102.MedPackPaddles
     @n102.Melee
     @n102.MeleeOneHand
     @n102.MountedTurret
     @n102.PaperBlueprint
     @n102.Pistol
     @n102.RemoteControl
     @n102.Rifle
     @n102.RocketLauncher
     @n102.SMG_P90
     @n102.Shotgun
     @n102.ShoulderLauncher
     @n102.SniperRifle
     @n102.TacticalShotgun
     @n102.TrapPlacement
     @n102.Unarmed
    ]
    Is DetectedDualWeild:double=0
   ) -> (Output_Get:double)
   > Set Is MeleeDualWeild_Bool(Is MeleeDualWeild_Bool:bool=false) -> (Output_Get:bool)
   > Set Is BowDetected(Is BowDetected:double=0) -> (Output_Get:double)
   Set Is BowDetected(execute=@n102.ExplosiveBow, Is BowDetected=1)
   Set Is DetectedDualWeild(execute=@n102.MeleeDualWield, Is DetectedDualWeild=1)
   > Set Is MeleeDualWeild_Bool(Is MeleeDualWeild_Bool=true)
   > Set Is BowDetected(Is BowDetected=0)
   n103 = Gravity Override State Function DEPRECATED(
    execute=@n5.then_2
    CharacterPartAnimInstance=Self-Reference()
    GameTimeInSeconds=GetGameTimeInSeconds()
    GravityOverrideParams=Get Gravity Override Params()
   )
   > Set Gravity Override(Gravity Override=@n103)
   n104 = Sequence!(execute=@n5.then_4) -> (then_0, then_1, then_2, then_3, then_4:Exec, then_5:Exec, then_6:Exec, then_7:Exec)
   n105 = GetOwnerSkeletalMeshComponent()
   n106 = BreakRotator(InRot=BreakTransform(InTransform=GetDeltaTransformFromRefPose(self=@n105, BoneName=head, BaseName=neck_01)).Rotation)
   n107 = Branch(execute=@n104.then_2, Condition="float < float"(A=@n106))
   > Set Alpha_isHeadRotateR(Alpha_isHeadRotateR:float=0) -> (Output_Get:float)
   > Set Alpha_isHeadRotateL(Alpha_isHeadRotateL:float=1) -> (Output_Get:float)
   n108 = AND Boolean!(A=NOT Boolean(A=Get bIsMoving2D()), B=Get bIsOnGround(), C:bool=NOT Boolean(A=Get bIsPlayingMeleeAnim())) -> (ReturnValue)
   > n109 = Branch(Condition=@n108)
   > Set Alpha_IsHeadRotateHoodCorrect_L(Alpha_IsHeadRotateHoodCorrect_L:double=1) -> (Output_Get:double)
   Set Alpha_IsHeadRotateHoodCorrect_L(execute=@n109.else, Alpha_IsHeadRotateHoodCorrect_L=0)
   Set Alpha_isHeadRotateR(execute=@n107.else, Alpha_isHeadRotateR=1)
   > Set Alpha_isHeadRotateL(Alpha_isHeadRotateL=0)
   > n110 = Branch(Condition=AND Boolean!(A="float > float"(A=@n106, B=25), B=@n108) -> (ReturnValue))
   > Set Alpha_IsHeadRotateHoodCorrect_R(Alpha_IsHeadRotateHoodCorrect_R:double=1) -> (Output_Get:double)
   Set Alpha_IsHeadRotateHoodCorrect_R(execute=@n110.else, Alpha_IsHeadRotateHoodCorrect_R=0)
   n111 = GetOwnerSkeletalMeshComponent()
   n112 = BreakRotator(InRot=BreakTransform(InTransform=GetDeltaTransformFromRefPose(self=@n111, BoneName=spine_02, BaseName=pelvis)).Rotation)
   n113 = Branch(
    execute=@n104.then_1
    Condition=AND Boolean!(A="float < float"(A=@n112.Yaw, B=8), B=Get AlphaIsMinigun_bool(), C:bool=Get bIsCrouching()) -> (ReturnValue)
   )
   > Set Alpha_IsMinugumAimUpper(Alpha_IsMinugumAimUpper:double=1) -> (Output_Get:double)
   Set Alpha_IsMinugumAimUpper(execute=@n113.else, Alpha_IsMinugumAimUpper=0)
   n114 = GetOwnerSkeletalMeshComponent()
   n115 = BreakRotator(InRot=BreakTransform(InTransform=GetDeltaTransformFromRefPose(self=@n114, BoneName=spine_02, BaseName=pelvis)).Rotation)
   n116 = Branch(execute=@n104.then_0, Condition=AND Boolean(A="float > float"(A=@n115.Yaw, B=12), B=Get AlphaIsMinigun_bool(), C=Get bIsCrouching()))
   > Set Alpha_isMinigunAimLower(Alpha_isMinigunAimLower:float=1) -> (Output_Get:float)
   Set Alpha_isMinigunAimLower(execute=@n116.else, Alpha_isMinigunAimLower=0)
   n117 = Branch(execute=@n104.then_3, Condition=AND Boolean!(A=Get AlphaIsMinigun_bool(), B=Get bIsCrouching()) -> (ReturnValue))
   > Set Alpha_IsMinugunAimLowerInCrouch(Alpha_IsMinugunAimLowerInCrouch:double=1) -> (Output_Get:double)
   Set Alpha_IsMinugunAimLowerInCrouch(execute=@n117.else, Alpha_IsMinugunAimLowerInCrouch=0)
   n118 = GetOwnerSkeletalMeshComponent()
   n119 = BreakRotator(InRot=BreakTransform(InTransform=GetDeltaTransformFromRefPose(self=@n118, BoneName=calf_l, BaseName=pelvis)).Rotation)
   n120 = Branch(execute=@n104.then_4, Condition=AND Boolean(A="float > float"(A=@n119, B=45), B=Get OnRailMode()))
   > Set Alpha_isRotateLegL_OnRail(Alpha_isRotateLegL_OnRail:float=1) -> (Output_Get:float)
   Set Alpha_isRotateLegL_OnRail(execute=@n120.else, Alpha_isRotateLegL_OnRail=0)
   n121 = GetOwnerSkeletalMeshComponent()
   n122 = BreakRotator(InRot=BreakTransform(InTransform=GetDeltaTransformFromRefPose(self=@n121, BoneName=spine_04, BaseName=pelvis)).Rotation)
   n123 = Branch(
    execute=@n104.then_5
    Condition=AND Boolean!(A="float < float"(A=@n122, B=-10), B=Get AlphaIsRocketLauncher(), C:bool=Get bIsCrouching()) -> (ReturnValue)
   )
   > Set Alpha_IsRocketLauncherCrouchAimUP(Alpha_IsRocketLauncherCrouchAimUP:double=1) -> (Output_Get:double)
   Set Alpha_IsRocketLauncherCrouchAimUP(execute=@n123.else, Alpha_IsRocketLauncherCrouchAimUP=0)
   n124 = OR Boolean(A=Get bIsTargeting(), B=Get bIsTargetingConsumableThrow() -> (bIsTargetingConsumableThrow:bool))
   n125 = GetDeltaTransformFromRefPose(self=GetOwnerSkeletalMeshComponent(), BoneName=neck_02, BaseName=head)
   n126 = "float >= float"(A=BreakRotator(InRot=BreakTransform(InTransform=@n125).Rotation).Yaw, B=2)
   Set TargetingUp(execute=@n5.then_7, TargetingUp:double="To Float (Boolean)"(InBool=AND Boolean!(A=@n124, B=@n126) -> (ReturnValue))) -> (Output_Get:double)
   n127 = CheckForBackbling()
   Set is Wearing Cape(execute=@n5.then_1, is Wearing Cape:bool=OR Boolean(A=@n127.bIsBackblingACape, B=@n127.bIsBackblingLong)) -> (Output_Get:bool)
   n128 = Event BlueprintProcessAnimInputEvent()
   > n129 = Branch(Condition=IsNamed(self=@n128.AnimInputEvent, InEventName=aie.Player.Zipline.Entered))
   > Set OnZipline(OnZipline:bool=true) -> (Output_Get:bool)
   > Set OnZiplineFloat(OnZiplineFloat:double=1) -> (Output_Get:double)
   > n130 = Branch(Condition=NOT Boolean(A=Get bRigidBodyEnabled()))
   > Set Is ZeplineRBDisable(Is ZeplineRBDisable:double=1) -> (Output_Get:double)
   Set Is ZeplineRBDisable(execute=@n130.else, Is ZeplineRBDisable=0)
   Branch(execute=@n129.else, Condition=IsNamed(self=@n128.AnimInputEvent, InEventName=aie.Player.Zipline.Exited))
   > Set OnZipline(OnZipline=false)
   > Set OnZiplineFloat(OnZiplineFloat=0)
   Event OnCharacterCustomizationLoaded()
   > n131 = Set SkelMesh(SkelMesh=GetPartSkeletalMeshComponent())
   > n132 = Get Dynamic Material Instance(SkeletalMeshComponent=@n131)
   > Set Mat Instance Dynamic(Mat Instance Dynamic=@n132)
   > n133 = GetOrCreateDynamicOverlayMaterial(MeshComponent=@n131, MaterialName=None)
   > Set Overlay Mat Instance Dynamic(Overlay Mat Instance Dynamic=@n133)
   AND Boolean(A=Get bIsParachuteOpened(), B=NOT Boolean(A=Get bIsMoving2D()))
   Reroute Node(InputPin:Wildcard) -> (OutputPin:Wildcard)
   "float > float"(A=@n112.Yaw, B=12)
  ]}
]
variables [
 Alpha_Crouch {type=number, default=1}
 Alpha_IsCrouchingIdle {type=number, default=1}
 Alpha_IsDBNOCarried {type=number}
 Alpha_IsDBNOMoving {type=number}
 Alpha_IsDBNOMovingRB Enabled {type=number}
 Alpha_IsDetectedAFK {type=number}
 Alpha_IsDisableRBFrontEnd Idle {type=number}
 Alpha_IsHeadRotateHoodCorrect_L {type=number}
 Alpha_IsHeadRotateHoodCorrect_R {type=number}
 Alpha_isHeadRotateL {type=number}
 Alpha_isHeadRotateR {type=number}
 Alpha_IsJump Disabled RB {type=number}
 Alpha_IsJumping {type=number, default=1}
 Alpha_IsJumpingIdle_RB Enabled {type=number}
 Alpha_IsJumpingWearingCape RB Enable {type=number}
 Alpha_IsMelee {type=number}
 Alpha_IsMelee Enable RB {type=number}
 Alpha_IsMeleeDualWeild {type=number}
 Alpha_IsMeleeMovingDisableRB {type=number}
 Alpha_IsMeleeSliding {type=number}
 Alpha_isMinigunAimLower {type=number}
 Alpha_IsMinugumAimUpper {type=number}
 Alpha_IsMinugunAimLowerInCrouch {type=number}
 Alpha_IsMinugunSelectedFloat {type=number}
 Alpha_IsRocketLauncherCrouchAimUP {type=number}
 Alpha_isRotateLegL_OnRail {type=number}
 Alpha_isSkydiveFloating {type=number}
 Alpha_isSkydiveFloating_RBDisable {type=number}
 Alpha_isSkydiveModeCape {type=number}
 Alpha_isSkydiveModeRBDisable {type=number}
 Alpha_IsSprinting {type=number, default=1}
 Alpha_IsWearingCapeCrouch {type=number}
 Alpha_IsWearingCapeEnable {type=number}
 Alpha_JumpMoving RB {type=number, default=1}
 Alpha_KneeSliding {type=number, default=1}
 Alpha_openParachute {type=number, default=1}
 Alpha_skyDiving {type=number, default=1}
 Alpha_SlidingWeapon {type=boolean}
 Alpha_WearingCapeRBEnable {type=number}
 AlphaIsAssaultRifle_Bool {type=boolean}
 AlphaIsAssaultRifle_Float {type=number}
 AlphaIsMinigun_bool {type=boolean}
 AlphaIsRocketLauncher {type=boolean}
 ar_firingPosition {type=integer}
 Body_Mesh {type=SkeletalMeshComponent}
 Gravity Override {type=Vector, default=<0 0 -980>}
 Gravity Override Params {
  type=GravityOverrideParamsStruct
  default {
   onGround_Moving {gravityOverride=<0 0 -1200>}
   rail_Grinding {windAmplitude=<800 800 800>, windFrequency=<30 30 30>}
   skydive_DivingDown {windAmplitude=<50 300 50>, windFrequency=<0.15 3 3>}
   skydive_FloatingDown {gravityOverride=<0 -500 100>, windAmplitude=<20 -30 30>, windFrequency=<20 20 15>}
   skydive_Parachuting {windAmplitude=<0 0 0>, windFrequency=<0 0 0>}
   ziplining {windAmplitude=<800 800 800>, windFrequency=<20 20 20>}
  }
 }
 Is BowDetected {type=number}
 Is CrouchMoving2d {type=number}
 Is DBNOIdle {type=number}
 Is DetectedDualWeild {type=number}
 Is Emote {type=number}
 Is FloatingFrontEnd {type=number}
 is LeftFallingInJump RB Enabled {type=number}
 Is Melee_moving2d {type=number}
 Is MeleeDisableRB {type=number}
 Is MeleeDualWeild_Bool {type=boolean}
 Is MinigunAimLowerNotCrouch {type=number}
 Is Moving2D RB Disable {type=number}
 Is ParachuteIdle {type=number}
 Is RightCrouch {type=number}
 is RightFallingInJump RB Enabled {type=number}
 is RocketLauncher_Float {type=number}
 Is RocketLauncherCorrect {type=number}
 Is RocketLauncherStand {type=number}
 Is SkyDivingFrontEnd {type=number}
 Is StandingIdle {type=number}
 is Wearing Cape {type=boolean}
 Is ZeplineRBDisable {type=number}
 Left {type=boolean}
 Local Noise Offset {type=Vector}
 Local Noise Transform Location {type=Vector}
 Mat Instance Dynamic {type=MaterialInstanceDynamic}
 OnRailMode {type=boolean}
 OnZipline {type=boolean}
 OnZiplineFloat {type=number}
 Overlay Mat Instance Dynamic {type=MaterialInstanceDynamic}
 PlayerPawn {type=FortPlayerPawnAthena}
 R1 Velocity Transform Location {type=Vector}
 R1 Velocity Transform Velocity {type=Vector}
 R2 Velocity Transform Location {type=Vector}
 R2 Velocity Transform Velocity {type=Vector}
 Right {type=boolean}
 RightBool / RB Enable {type=number}
 RightCrouchMele {type=number}
 Sim Space Settings {
  type=SimSpaceSettings
  default {
   dampingAlpha=1
   maxAngularAcceleration=10000
   maxAngularVelocity=10000
   maxLinearAcceleration=10000
   maxLinearVelocity=10000
   velocityScaleZ=1
   worldAlpha=0
   externalLinearDragV=<0 0 0>
  }
 }
 SkelMesh {type=SkeletalMeshComponent}
 TargetingUp {type=number, default=1}
 Velocity Interp Speed {type=number}
 WPO Pho Turbulence {type=number}
 WPO Pho Wind Intensity {type=number}
]
